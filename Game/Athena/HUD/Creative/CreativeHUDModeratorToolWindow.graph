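# Creative HUD activatable window for moderator tool mode, showing a moderator timer info widget and a Back button.
# It starts collapsed and follows the player controller's mode activated/deactivated delegates, guarded by its active state.
# Activation regenerates timer info, shows the cursor, blocks look/move input and adds an Enhanced Input mapping context; deactivation hides it and reverts the input changes.
# Back button or input action asks the controller to exit, rather than closing directly.
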WidgetBlueprint /Game/Athena/HUD/Creative/CreativeHUDModeratorToolWindow
graphs [EventGraph {nodes [
   n1 = "On Clicked (Button_Back)"!() -> (OutputDelegate, then, Button:Common Button Base Object Reference) {is=title K2Node_ComponentBoundEvent}
   SetCreativeModeratorModeActivateState!(
    execute=[@n1.then, @n7.Triggered]
    self:Fort Player Controller Gameplay Object Reference=Get PlayerController() -> (PlayerController:Fort Player Controller Gameplay Object Reference)
    bIsActive:bool=false
   ) -> (then)
   n2 = ActivateModeratorToolEvent() {is=title K2Node_CustomEvent}
   > n3 = Branch(Condition=Get bIsActive() -> (bIsActive:bool))
   SetVisibility(execute=@n3.else, InVisibility=SelfHitTestInvisible)
   > ActivateWidget()
   > GenerateTimerInfo(self:Moderator Timer Info Widget Object Reference=Get WBP_ModeratorTimerInfo() -> (WBP_ModeratorTimerInfo:WBP Moderator Timer Info Object Reference))
   n4 = Get PlayerController()
   > Set bShowMouseCursor!(execute, bShowMouseCursor:bool=true, self:Player Controller Object Reference=@n4) -> (then, Output_Get:bool)
   > SetIgnoreLookInput!(execute, self:Controller Object Reference=@n4, bNewLookInput:bool=true) -> (then)
   > SetIgnoreMoveInput!(execute, self:Controller Object Reference=@n4, bNewMoveInput:bool=true) -> (then)
   > AddMappingContext!(
    execute:Exec
    self:Object Reference=Get EnhancedInputLocalPlayerSubsystem!(PlayerController:Player Controller Object Reference=@n4) -> (ReturnValue:Enhanced Input Local Player Subsystem Object Reference) {is=title K2Node_GetSubsystemFromPC}
    MappingContext:Input Mapping Context Object Reference=/Game/Athena/HUD/Creative/Input/Creative_ModeratorTool_InputMappingContext
    Priority:int=1
    Options:"Modify Context Options Structure (by ref)"=(bIgnoreAllPressedKeysUntilRelease=True,bForceImmediately=False,bNotifyUserSettings=False)
   ) -> (then:Exec) {is=title K2Node_Message}
   n5 = DeactivateModeratorToolEvent() {is=title K2Node_CustomEvent}
   > Branch(Condition=Get bIsActive())
   > SetVisibility(InVisibility=Collapsed)
   > DeactivateWidget()
   n6 = Get PlayerController()
   > Set bShowMouseCursor(bShowMouseCursor=false, self=@n6)
   > SetIgnoreLookInput(self=@n6, bNewLookInput=false)
   > SetIgnoreMoveInput(self=@n6, bNewMoveInput=false)
   > RemoveMappingContext!(
    execute
    self:Enhanced Input Subsystem Interface Interface=Get EnhancedInputLocalPlayerSubsystem(PlayerController=@n6)
    MappingContext:Input Mapping Context Object Reference=&Creative_ModeratorTool_InputMappingContext
    Options:"Modify Context Options Structure (by ref)"=(bIgnoreAllPressedKeysUntilRelease=True,bForceImmediately=False,bNotifyUserSettings=False)
   ) -> (then)
   n7 = EnhancedInputAction Creative_ModeratorTool_BackInputAction!() -> (
    Triggered:Exec
    Started:Exec
    Ongoing:Exec
    Canceled:Exec
    Completed:Exec
    ActionValue:bool
    ElapsedSeconds:double=~
    TriggeredSeconds:double=~
    InputAction:Input Action Object Reference=/Game/Athena/HUD/Creative/Input/Creative_ModeratorTool_BackInputAction
   ) {is=title K2Node_EnhancedInputAction}
   Event Construct()
   > SetVisibility(InVisibility=Collapsed)
   > n8 = Cast To FortPlayerControllerGameplay!(execute, Object=GetOwningPlayer()) -> (then, CastFailed, AsFort Player Controller Gameplay:Fort Player Controller Gameplay Object Reference) {is=title K2Node_DynamicCast}
   > n9 = Set PlayerController(PlayerController:Fort Player Controller Gameplay Object Reference=@n8) -> (Output_Get:Fort Player Controller Gameplay Object Reference)
   > Bind Event to On Creative Moderator Tool Mode Activated(self:Fort Player Controller Gameplay Object Reference=@n9, Delegate=@n2) {is=title K2Node_AddDelegate}
   > Bind Event to On Creative Moderator Tool Mode Deactivated(self:Fort Player Controller Gameplay Object Reference=Get PlayerController(), Delegate=@n5) {is=title K2Node_AddDelegate}
   Event Destruct()
   > Unbind Event from On Creative Moderator Tool Mode Activated(self:Fort Player Controller Gameplay Object Reference=Get PlayerController(), Delegate=@n2) {is=title K2Node_RemoveDelegate}
   > Unbind Event from On Creative Moderator Tool Mode Deactivated(self:Fort Player Controller Gameplay Object Reference=Get PlayerController(), Delegate=@n5) {is=title K2Node_RemoveDelegate}
  ]}]
variables [PlayerController {type=FortPlayerControllerGameplay, default=null}]
widgets {
 info {inheritedWidgetCount=0, parentClass=&CommonActivatableWidget, rootWidgetClass=&Overlay, widgetCount=7}
 widgets [
  Overlay_0 {widgetClassPath=&Overlay, widget=Overlay_0}
  Overlay {
   widgetClassPath=&Overlay
   bIsVariable=true
   parent=Overlay_0
   widget=Overlay
   slot {class=OverlaySlot, properties {horizontalAlignment=HAlign_Fill, verticalAlignment=VAlign_Fill}}
  }
  WBP_ModeratorTimerInfo {
   widgetClassPath=/Game/Athena/HUD/Creative/Widgets/WBP_ModeratorTimerInfo
   bIsVariable=true
   parent=Overlay
   widget=WBP_ModeratorTimerInfo
   slot {class=OverlaySlot}
  }
  Switcher_Input {
   widgetClassPath=&WidgetSwitcher
   bIsVariable=true
   parent=Overlay_0
   widget=Switcher_Input
   slot {class=OverlaySlot, properties {horizontalAlignment=HAlign_Fill, verticalAlignment=VAlign_Bottom}}
  }
  MainInput {
   widgetClassPath=&Overlay
   parent=Switcher_Input
   widget=MainInput
   slot {class=WidgetSwitcherSlot, properties {verticalAlignment=VAlign_Bottom, padding=<ltrb 12 12 12 12>}}
  }
  HorizontalBox_0 {
   widgetClassPath=&HorizontalBox
   parent=MainInput
   widget=HorizontalBox_0
   slot {class=OverlaySlot, properties {horizontalAlignment=HAlign_Center}}
  }
  Button_Back {
   widgetClassPath=/FortUILibrary/UIKit/Buttons/WBP_UIKit_Button_Quiet
   bIsVariable=true
   parent=MainInput
   widget=Button_Back
   properties {minHeight=48, text=Back, triggeringInputAction {dataTable=&InputActionDataTable, rowName=PhotoModeBackRetake}}
   slot {class=OverlaySlot, properties {horizontalAlignment=HAlign_Right, verticalAlignment=VAlign_Fill}}
  }
 ]
}
