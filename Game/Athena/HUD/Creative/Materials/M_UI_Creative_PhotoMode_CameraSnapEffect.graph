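# Masked UI material for the Creative photo mode camera snap effect, shaped as a circular shutter that closes and reopens.
# The CurrentTime parameter becomes a triangle ramp (min of t and 1-t, doubled) that shrinks the radius to zero at midpoint and restores it by the end.
# That radius drives an exponential radial gradient on UVs from MF_UI_UVScaling, which is inverted and truncated into a hard-edged mask outside the circle.
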Material /Game/Athena/HUD/Creative/Materials/M_UI_Creative_PhotoMode_CameraSnapEffect
graph {
 properties {blendMode=BLEND_Masked, materialDomain=MD_UI, pixelDepthOffsetMode=PDOM_Legacy, usedShadingModels=Default Lit}
 nodes [
  n1 = MaterialFunctionCall() {materialFunction=&MF_UI_UVScaling}
  n2 = ScalarParameter() {parameterName=CurrentTime}
  n3 = OneMinus(None=@n2)
  n4 = LinearInterpolate(Alpha=Multiply(A=Min(A=@n3, B=@n2)) {constB=2}) {constA=1.12, constB=0}
  Truncate!(None=OneMinus(None=MaterialFunctionCall(UVs=@n1.Scaled to Fit, Radius=@n4) {materialFunction=&RadialGradientExponential}.RadialGradientExponential)) {is=class MaterialExpressionTruncate}
 ]
}
