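# Creative photo mode HUD overlay, an activatable widget that hides the HUD while showing guides and capture controls.
# A switcher alternates Capture and Confirm/Retake rows, freezing move/look input during confirmation.
# Buttons and component-routed input actions share handlers calling FortControllerComponent_PhotoMode to save camera settings, accept, cancel or restore the last position via hold with material progress and audio.
# Render completion plays a material snap animation, then returns to capture.
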WidgetBlueprint /Game/Athena/HUD/Creative/CreativeHUDPhotoModeOverlay
animations [CameraSnapAnimation {
  binding_count=1
  length_seconds=1
  object=&self:CameraSnapAnimation
  sequence {
   bindings [CameraSnapEffect {
     id=#1
     possessed_class=Image
     tracks [{
       type=MovieSceneWidgetMaterialTrack
       sections [{type=MovieSceneParameterSection, channels [CurrentTime {type=MovieSceneScriptingFloatChannel, keys [[0,0],[60,1]]}]}]
      }]
    }]
   display_rate {numerator=60}
  }
 }]
graphs [
 EventGraph {nodes [
   n1 = "On Button Clicked (Button_Capture)"!() -> (OutputDelegate, then, Button:Common Button Legacy Object Reference) {is=title K2Node_ComponentBoundEvent}
   SetOverlayVisibility!(execute=[@n1.then, @n11.then, @n16.then], self, bSetVisible:bool=false) -> (then)
   > SetActiveWidgetIndex(self=Get Switcher_Input() -> (Switcher_Input:Widget Switcher Object Reference), Index=1)
   > ToggleInput!(execute, self, ToggleOn:bool=false) -> (then)
   n2 = "On Button Clicked (Button_Overlay)"!() -> (OutputDelegate, then, Button:Common Button Legacy Object Reference) {is=title K2Node_ComponentBoundEvent}
   n3 = Get Overlay() -> (Overlay:Overlay Object Reference)
   n4 = Set bIsOverlayVisible(execute=[@n2.then, @n13.then, @n19.then], bIsOverlayVisible:bool=NOT Boolean(A=IsVisible(self=@n3))) -> (Output_Get:bool)
   > SetOverlayVisibility(bSetVisible=@n4)
   n5 = "On Clicked (Button_Back)"!() -> (OutputDelegate, then, Button:Common Button Base Object Reference) {is=title K2Node_ComponentBoundEvent}
   ServerCancelPhotoMode(
    execute=[@n5.then, @n9.then, @n14.then]
    self:Fort Controller Component Photo Mode Object Reference=Get PhotoModeComponent() -> (PhotoModeComponent:Fort Controller Component Photo Mode Object Reference)
   )
   > SetVisibility(InVisibility=Collapsed)
   > SetHUDElementVisibility!(
    execute
    self:Fort HUDContext Object Reference=Get FortHUDContext!() -> (ReturnValue:Fort HUDContext Object Reference) {is=title K2Node_GetBlueprintContextEx}
    HUDElementTags:"Gameplay Tag Container Structure (by ref)"=MakeGameplayTagContainerFromTag!(SingleTag:Gameplay Tag Structure=(TagName="HUD")) -> (ReturnValue:Gameplay Tag Container Structure)
    bHideElements:bool=false
    ReasonName:"Name (by ref)"=None
   ) -> (then)
   > Set bIsPhotoModeActive(bIsPhotoModeActive:bool=false) -> (Output_Get:bool)
   > SetCreativePhotoModeActivateState!(
    execute
    self:Fort Player Controller Gameplay Object Reference=Get PlayerController() -> (PlayerController:Fort Player Controller Gameplay Object Reference)
    bIsActive:bool=false
   ) -> (then)
   > DeactivateWidget()
   n6 = "On Clicked (Button_Confirm)"!() -> (OutputDelegate, then, Button:Common Button Base Object Reference) {is=title K2Node_ComponentBoundEvent}
   ServerSaveCameraSettings(execute=[@n6.then, @n10.then, @n17.then], self:Fort Controller Component Photo Mode Object Reference=Get PhotoModeComponent())
   > SetVisibility(self=Get Switcher_Input(), InVisibility=Hidden)
   > ClientAcceptPhoto(self:Fort Controller Component Photo Mode Object Reference=Get PhotoModeComponent())
   n7 = "On Clicked (Button_Retake)"!() -> (OutputDelegate, then, Button:Common Button Base Object Reference) {is=title K2Node_ComponentBoundEvent}
   SetOverlayVisibility(execute=[@n7.then, @n8.then, @n15.then], bSetVisible=Get bIsOverlayVisible() -> (bIsOverlayVisible:bool))
   > SetActiveWidgetIndex(self=Get Switcher_Input(), Index=0)
   > ToggleInput(ToggleOn=true)
   n8 = "On Button Clicked (Button_Retake)"!() -> (OutputDelegate, then, Button:Common Button Legacy Object Reference) {is=title K2Node_ComponentBoundEvent}
   n9 = "On Button Clicked (Button_Back)"!() -> (OutputDelegate, then, Button:Common Button Legacy Object Reference) {is=title K2Node_ComponentBoundEvent}
   n10 = "On Button Clicked (Button_Confirm)"!() -> (OutputDelegate, then, Button:Common Button Legacy Object Reference) {is=title K2Node_ComponentBoundEvent}
   n11 = "On Clicked (Button_Capture)"!() -> (OutputDelegate, then, Button:Common Button Base Object Reference) {is=title K2Node_ComponentBoundEvent}
   n12 = "On Hold Completed (Button_LastSettings)"() {is=title K2Node_ComponentBoundEvent}
   ServerApplyPhotoSettingsToPlayer(execute=[@n12.then, @n18.then], self:Fort Controller Component Photo Mode Object Reference=Get PhotoModeComponent())
   n13 = "On Clicked (Button_Overlay)"!() -> (OutputDelegate, then, Button:Common Button Base Object Reference) {is=title K2Node_ComponentBoundEvent}
   n14 = Back_KeyPressed() {is=title K2Node_CustomEvent}
   n15 = Retake_KeyPressed() {is=title K2Node_CustomEvent}
   n16 = Capture_KeyPressed() {is=title K2Node_CustomEvent}
   n17 = Confirm_KeyPressed() {is=title K2Node_CustomEvent}
   n18 = LastSettings_KeyHeld() {is=title K2Node_CustomEvent}
   n19 = Overlay_KeyPressed() {is=title K2Node_CustomEvent}
   Event Destruct()
   > Is Valid(InputObject=Get PlayerController())
   > K2Node_Composite_3 = Bindings!(Destruct:Exec, Construct:Exec=@n22.then) {is=title K2Node_Composite}
   Event Construct()
   > SetVisibility(InVisibility=Collapsed)
   > n20 = Cast To FortPlayerControllerGameplay!(execute, Object=GetOwningPlayer()) -> (then, CastFailed, AsFort Player Controller Gameplay:Fort Player Controller Gameplay Object Reference) {is=title K2Node_DynamicCast}
   > Set PlayerController(PlayerController:Fort Player Controller Gameplay Object Reference=@n20) -> (Output_Get:Fort Player Controller Gameplay Object Reference)
   > n21 = GetDynamicMaterial(self=Get HoldProgressBar!(self:WBP UIKit Button CTA Base Object Reference=Get Button_LastSettings() -> (Button_LastSettings:WBP UIKit Button Quiet Object Reference)) -> (HoldProgressBar:Image Object Reference))
   > n22 = Set LastSettingsHoldMaterial(LastSettingsHoldMaterial:Material Instance Dynamic Object Reference=@n21) -> (Output_Get:Material Instance Dynamic Object Reference)
   Event On Deactivated()
   > n23 = Cast To FortPlayerControllerGameplay(Object=GetOwningPlayer())
   > n24 = Branch(Condition=IsCreativeModeratorModeActive!(self:Fort Player Controller Gameplay Object Reference=@n23) -> (ReturnValue:bool))
   n25 = GetComponentByClass!(self:Actor Object Reference=@n23, ComponentClass:Actor Component Class Reference=/Script/FortniteGame.FortControllerComponent_PhotoMode) -> (ReturnValue:Fort Controller Component Photo Mode Object Reference)
   Is Valid(exec=@n24.else, InputObject=@n25)
   > RemovePhotomodeOverlayWidgetScene(self:Fort Controller Component Photo Mode Object Reference=@n25)
  ]}
 Bindings {
  path=EventGraph.K2Node_Composite_3.Bindings
  nodes [
   n1 = PhotoFinishedRendering() {is=title K2Node_CustomEvent}
   > PlayAnimation(InAnimation=Get CameraSnapAnimation() -> (CameraSnapAnimation:Widget Animation Object Reference), PlaybackSpeed=1.5)
   n2 = PhotoModeActivated() {is=title K2Node_CustomEvent}
   > K2Node_Composite_0 = Activate() {is=title K2Node_Composite}
   n3 = SnapEffectFinished() {is=title K2Node_CustomEvent}
   > Retake_KeyPressed() {is=title K2Node_CallFunction}
   > SetVisibility(self=Get Switcher_Input(), InVisibility=SelfHitTestInvisible)
   n4 = HandleInputAction!() -> (
    OutputDelegate
    then
    InputAction_SourceAction:Input Action Object Reference
    InputAction_TriggerEvent:ETriggerEvent Enum
    InputAction_LastTriggeredWorldTime:float=~
    InputAction_Triggers:Array of Input Trigger Object References
    InputAction_Modifiers:Array of Input Modifier Object References
    InputAction_ElapsedProcessedTime:float=~
    InputAction_ElapsedTriggeredTime:float=~
   ) {is=title K2Node_CustomEvent}
   > n5 = Switch on Name!(
    execute:Exec
    Selection:Name=Get Name!(self:Player Mappable Key Settings Object Reference=Get PlayerMappableKeySettings!(self:Input Action Object Reference=@n4.InputAction_SourceAction) -> (PlayerMappableKeySettings:Player Mappable Key Settings Object Reference)) -> (Name:Name=None)
   ) -> (IA_UI_PrimaryForward:Exec, Creative_PhotoMode_LastSettings:Exec, Creative_PhotoMode_ToggleOverlay:Exec, IA_UI_PrimaryBack:Exec) {
    is=title K2Node_SwitchName
    bHasDefaultPin=false
    pinNames [IA_UI_PrimaryForward, Creative_PhotoMode_LastSettings, Creative_PhotoMode_ToggleOverlay, IA_UI_PrimaryBack]
   }
   K2Node_Composite_9 = LastSettings!(
    execute=@n5.Creative_PhotoMode_LastSettings
    Trigger Event:ETriggerEvent Enum=@n4.InputAction_TriggerEvent
    " Elapsed Time":float=@n4.InputAction_ElapsedProcessedTime
   ) {is=title K2Node_Composite}
   K2Node_Composite_3 = ToggleOverlay(execute=@n5.Creative_PhotoMode_ToggleOverlay) {is=title K2Node_Composite}
   K2Node_Composite_2 = BackRetake(execute=@n5.IA_UI_PrimaryBack) {is=title K2Node_Composite}
   > K2Node_Composite_1 = CaptureConfirm() {is=title K2Node_Composite}
   n6 = SavedCameraFound() {is=title K2Node_CustomEvent}
   > SetVisibility(self=Get Button_LastSettings(), InVisibility=SelfHitTestInvisible)
   n7 = Inputs!() -> (Destruct:Exec, Construct:Exec)
   n8 = Get PlayerController()
   > n9 = Unbind Event from On Creative Photo Mode Activated(self:Fort Player Controller Gameplay Object Reference=@n8, Delegate=@n2) {is=title K2Node_RemoveDelegate}
   n10 = Set PhotoModeComponent(
    execute=@n16.then
    PhotoModeComponent:Fort Controller Component Photo Mode Object Reference=GetComponentByClass(self=Get PlayerController(), ComponentClass=&FortControllerComponent_PhotoMode)
   ) -> (Output_Get:Fort Controller Component Photo Mode Object Reference)
   Unbind Event from Saved Camera Found Callback(execute=@n9.then, self:Fort Controller Component Photo Mode Object Reference=@n10, Delegate=@n6) {is=title K2Node_RemoveDelegate}
   n11 = Create Event()
   n12 = Get PhotoModeComponent()
   > Unbind Event from Forced Exit Callback(self:Fort Controller Component Photo Mode Object Reference=@n12, Delegate=@n11) {is=title K2Node_RemoveDelegate}
   n13 = Get PhotoModeComponent()
   > Unbind Event from Handle Input Action Callback(self:Fort Controller Component Photo Mode Object Reference=@n13, Delegate=@n4) {is=title K2Node_RemoveDelegate}
   n14 = Get PhotoModeComponent()
   > Unbind Event from Photo Finished Rendering Callback(self:Fort Controller Component Photo Mode Object Reference=@n14, Delegate=@n1) {is=title K2Node_RemoveDelegate}
   n15 = Get CameraSnapAnimation()
   > UnbindFromAnimationFinished!(execute, self:User Widget Object Reference, Animation:Widget Animation Object Reference=@n15, Delegate:Delegate=@n3) -> (then)
   n16 = Bind Event to On Creative Photo Mode Activated(execute=@n7.Construct, self:Fort Player Controller Gameplay Object Reference=@n8, Delegate=@n2) {is=title K2Node_AddDelegate}
   BindToAnimationFinished(execute=@n17.then, Animation=@n15, Delegate=@n3)
   Outputs()
   n17 = Bind Event to Photo Finished Rendering Callback(execute=@n18.then, self:Fort Controller Component Photo Mode Object Reference=@n14, Delegate=@n1) {is=title K2Node_AddDelegate}
   n18 = Bind Event to Handle Input Action Callback(execute=@n19.then, self:Fort Controller Component Photo Mode Object Reference=@n13, Delegate=@n4) {is=title K2Node_AddDelegate}
   Bind Event to Saved Camera Found Callback(execute=@n10.then, self:Fort Controller Component Photo Mode Object Reference=@n10, Delegate=@n6) {is=title K2Node_AddDelegate}
   > n19 = Bind Event to Forced Exit Callback(self:Fort Controller Component Photo Mode Object Reference=@n12, Delegate=@n11) {is=title K2Node_AddDelegate}
  ]
 }
 Activate {
  path=EventGraph.K2Node_Composite_3.Bindings.K2Node_Composite_0.Activate
  nodes [
   Inputs!() -> (execute:Exec)
   > SetVisibility(self=Get Button_LastSettings(), InVisibility=Collapsed)
   > SetVisibility(InVisibility=SelfHitTestInvisible)
   > Set bIsPhotoModeActive(bIsPhotoModeActive=true)
   > SetHUDElementVisibility(self=Get FortHUDContext(), HUDElementTags=MakeGameplayTagContainerFromTag(SingleTag=(TagName="HUD")), bHideElements=true, ReasonName=None)
   > ServerSetupPhotoMode(self:Fort Controller Component Photo Mode Object Reference=Get PhotoModeComponent())
   > ActivateWidget()
   Outputs()
  ]
 }
 CaptureConfirm {
  path=EventGraph.K2Node_Composite_3.Bindings.K2Node_Composite_1.CaptureConfirm
  nodes [
   Inputs()
   > n1 = Switch on Int!(execute:Exec, Selection:int=GetActiveWidgetIndex!(self:Widget Switcher Object Reference=Get Switcher_Input()) -> (ReturnValue:int=0)) -> (0:Exec, 1:Exec) {bHasDefaultPin=false}
   > Capture_KeyPressed() {is=title K2Node_CallFunction}
   > PlaySound2D(Sound=Get PressedSound!(self:WBP UIKit Button CTA Base Object Reference=Get Button_Capture() -> (Button_Capture:WBP UIKit Button Loud Object Reference)) -> (PressedSound:Sound Base Object Reference))
   Confirm_KeyPressed(execute=@n1.1) {is=title K2Node_CallFunction}
   > PlaySound2D(Sound=Get PressedSound(self=Get Button_Confirm() -> (Button_Confirm:WBP UIKit Button Loud Object Reference)))
   Outputs()
  ]
 }
 BackRetake {
  path=EventGraph.K2Node_Composite_3.Bindings.K2Node_Composite_2.BackRetake
  nodes [
   Inputs()
   > n1 = Switch on Int(Selection=GetActiveWidgetIndex(self=Get Switcher_Input()) -> (ReturnValue=0)) {bHasDefaultPin=false}
   > Back_KeyPressed() {is=title K2Node_CallFunction}
   > PlaySound2D(Sound=Get PressedSound(self=Get Button_Back() -> (Button_Back:WBP UIKit Button Quiet Object Reference)))
   Retake_KeyPressed(execute=@n1.1)
   > PlaySound2D(Sound=Get PressedSound(self=Get Button_Retake() -> (Button_Retake:WBP UIKit Button Quiet Object Reference)))
   Outputs()
  ]
 }
 ToggleOverlay {
  path=EventGraph.K2Node_Composite_3.Bindings.K2Node_Composite_3.ToggleOverlay
  nodes [
   Inputs()
   > n1 = Branch(Condition=Get bIsInputIgnored() -> (bIsInputIgnored:bool))
   Overlay_KeyPressed(execute=@n1.else) {is=title K2Node_CallFunction}
   > PlaySound2D(Sound=Get PressedSound(self=Get Button_Overlay() -> (Button_Overlay:WBP UIKit Button Quiet Object Reference)))
   Outputs()
  ]
 }
 LastSettings {
  path=EventGraph.K2Node_Composite_3.Bindings.K2Node_Composite_9.LastSettings
  nodes [
   n1 = HoldPressedFinished() {is=title K2Node_CustomEvent}
   > RemoveAudioComponent!(
    execute
    self
    Component:"Audio Component Object Reference (by ref)"=Get HoldAudioPressedComponent() -> (HoldAudioPressedComponent:Audio Component Object Reference)
   ) -> (then)
   > n2 = CreateSound2D!(
    execute
    Sound:Sound Base Object Reference=Get HoldSound!(self:WBP UIKit Button CTA Base Object Reference=Get Button_LastSettings()) -> (HoldSound:Sound Base Object Reference)
    VolumeMultiplier:float=1
    PitchMultiplier:float=1
    StartTime:float=0
    ConcurrencySettings:Sound Concurrency Object Reference
    bPersistAcrossLevelTransition:bool=false
    bAutoDestroy:bool=false
   ) -> (then, ReturnValue:Audio Component Object Reference)
   > n3 = Set HoldAudioComponent(HoldAudioComponent:Audio Component Object Reference=@n2) -> (Output_Get:Audio Component Object Reference)
   n4 = LoopHoldSound() {is=title K2Node_CustomEvent}
   n5 = Assign On Audio Finished!(execute:Exec=@n3.then, self:Audio Component Object Reference=@n3, Delegate:"Delegate (by ref)"=@n4) -> (then:Exec) {is=title K2Node_AssignDelegate}
   Play!(
    execute=[@n5.then, @n4.then]
    self:Audio Component Object Reference=Get HoldAudioComponent() -> (HoldAudioComponent:Audio Component Object Reference)
    StartTime:float=0
   ) -> (then)
   n6 = Inputs!() -> (execute, Trigger Event:ETriggerEvent Enum=None, " Elapsed Time":float=~)
   > n7 = Branch(Condition=Get bIsInputIgnored())
   n8 = Switch on ETriggerEvent!(execute=@n7.else, Selection:ETriggerEvent Enum=@n6) -> (Triggered:Exec, Started:Exec, Ongoing:Exec, Canceled:Exec, Completed:Exec) {is=title K2Node_SwitchEnum}
   n9 = PlaySound2D(
    execute=@n8.Canceled
    Sound=Get HoldReleasedSound!(self:WBP UIKit Button CTA Base Object Reference=Get Button_LastSettings()) -> (HoldReleasedSound:Sound Base Object Reference)
   )
   n10 = Is Valid(exec=[@n9.then, @n14.then], InputObject=Get HoldAudioPressedComponent())
   > n11 = RemoveAudioComponent(Component=Get HoldAudioPressedComponent())
   n12 = Is Valid(exec=[@n11.then, @n10.Is Not Valid], InputObject=Get HoldAudioComponent())
   > n13 = RemoveAudioComponent(Component=Get HoldAudioComponent())
   SetScalarParameterValue(
    execute=[@n13.then, @n12.Is Not Valid]
    self=Get LastSettingsHoldMaterial() -> (LastSettingsHoldMaterial:Material Instance Dynamic Object Reference)
    ParameterName=ShapeSwitch
    Value=0
   )
   LastSettings_KeyHeld(execute=@n8.Completed) {is=title K2Node_CallFunction}
   > n14 = PlaySound2D(Sound=Get HoldCompletedSound!(self:WBP UIKit Button CTA Base Object Reference=Get Button_LastSettings()) -> (HoldCompletedSound:Sound Base Object Reference))
   n15 = GetRequiredHoldTime!(self:Common Button Base Object Reference=Get Button_LastSettings()) -> (ReturnValue:float)
   SetScalarParameterValue(
    execute=@n8.Ongoing
    self=Get LastSettingsHoldMaterial()
    ParameterName=ShapeSwitch
    Value="Clamp (Float)"(Value=float / float(A=@n6." Elapsed Time", B=@n15), Min=0, Max=1)
   )
   n16 = CreateSound2D(
    execute=@n8.Started
    Sound=Get HoldPressedSound!(self:WBP UIKit Button CTA Base Object Reference=Get Button_LastSettings()) -> (HoldPressedSound:Sound Base Object Reference)
    VolumeMultiplier=1
    PitchMultiplier=1
    StartTime=0
    bPersistAcrossLevelTransition=false
    bAutoDestroy=false
   )
   > n17 = Set HoldAudioPressedComponent(HoldAudioPressedComponent:Audio Component Object Reference=@n16) -> (Output_Get:Audio Component Object Reference)
   > Assign On Audio Finished(self=@n17, Delegate=@n1)
   > Play(self=Get HoldAudioPressedComponent(), StartTime=0)
   Outputs()
  ]
 }
 RemoveAudioComponent {nodes [
   n1 = RemoveAudioComponent!() -> (then, Component:"Audio Component Object Reference (by ref)") {is=title K2Node_FunctionEntry}
   > Unbind all Events from On Audio Finished!(execute:Exec, self:Audio Component Object Reference=@n1) -> (then:Exec) {is=title K2Node_ClearDelegate}
   > Stop(self:Audio Component Object Reference=Get Component() -> (Component:"Audio Component Object Reference (by ref)"))
   > Destroy Component(self:Actor Component Object Reference=Get Component())
   > "Set Audio Component Object Reference (by ref)"!(execute:Exec, Target:"Audio Component Object Reference (by ref)"=@n1, Value:Audio Component Object Reference) -> (then:Exec) {is=title K2Node_VariableSetRef}
  ]}
 SetOverlayVisibility {nodes [
   n1 = SetOverlayVisibility!() -> (then, bSetVisible:bool=~) {is=title K2Node_FunctionEntry}
   > n2 = Branch(Condition=@n1)
   n3 = Get Overlay()
   > SetVisibility(self=@n3, InVisibility=SelfHitTestInvisible)
   SetVisibility(execute=@n2.else, self=@n3, InVisibility=Collapsed)
  ]}
 ToggleInput {nodes [
   n1 = ToggleInput!() -> (then, ToggleOn:bool=~) {is=title K2Node_FunctionEntry}
   > n2 = Branch(Condition=@n1)
   > Branch(Condition=Get bIsInputIgnored())
   n3 = Get PlayerController()
   > SetIgnoreLookInput!(execute, self:Controller Object Reference=@n3, bNewLookInput:bool=false) -> (then)
   > SetIgnoreMoveInput!(execute, self:Controller Object Reference=@n3, bNewMoveInput:bool=false) -> (then)
   > Set bIsInputIgnored(bIsInputIgnored:bool=false) -> (Output_Get:bool)
   Branch(execute=@n2.else, Condition=NOT Boolean(A=Get bIsInputIgnored()))
   > SetIgnoreLookInput(self=@n3, bNewLookInput=true)
   > SetIgnoreMoveInput(self=@n3, bNewMoveInput=true)
   > Set bIsInputIgnored(bIsInputIgnored=true)
  ]}
]
variables [
 bIsInputIgnored {type=boolean}
 bIsOverlayVisible {type=boolean, default=true}
 bIsPhotoModeActive {type=boolean}
 HoldAudioComponent {type=AudioComponent, default=null}
 HoldAudioPressedComponent {type=AudioComponent, default=null}
 LastSettingsHoldMaterial {type=MaterialInstanceDynamic}
 PhotoModeComponent {type=FortControllerComponent_PhotoMode, default=null}
 PlayerController {type=FortPlayerControllerGameplay, default=null}
]
widgets {
 info {inheritedWidgetCount=0, parentClass=&CommonActivatableWidget, rootWidgetClass=&Overlay, widgetCount=15}
 widgets [
  Overlay_0 {widgetClassPath=&Overlay, widget=Overlay_0}
  Overlay {
   widgetClassPath=&Overlay
   bIsVariable=true
   parent=Overlay_0
   widget=Overlay
   slot {class=OverlaySlot, properties {horizontalAlignment=HAlign_Fill, verticalAlignment=VAlign_Fill}}
  }
  WBP_PhotoMode_Guideline {
   widgetClassPath=&/WBP_PhotoMode_Guideline
   bIsVariable=true
   parent=Overlay
   widget=WBP_PhotoMode_Guideline
   slot {class=OverlaySlot, properties {horizontalAlignment=HAlign_Fill, verticalAlignment=VAlign_Fill}}
  }
  WBP_PhotoMode_Overlay_1 {
   widgetClassPath=&/WBP_PhotoMode_Overlay_1
   bIsVariable=true
   parent=Overlay
   widget=WBP_PhotoMode_Overlay_1
   properties {visibility=Collapsed}
   slot {class=OverlaySlot, properties {horizontalAlignment=HAlign_Fill, verticalAlignment=VAlign_Fill}}
  }
  Switcher_Input {
   widgetClassPath=&WidgetSwitcher
   bIsVariable=true
   parent=Overlay_0
   widget=Switcher_Input
   slot {class=OverlaySlot, properties {horizontalAlignment=HAlign_Fill, verticalAlignment=VAlign_Bottom}}
  }
  MainInput {
   widgetClassPath=&Overlay
   parent=Switcher_Input
   widget=MainInput
   slot {class=WidgetSwitcherSlot, properties {verticalAlignment=VAlign_Bottom, padding=<ltrb 12 12 12 12>}}
  }
  HorizontalBox_0 {
   widgetClassPath=&HorizontalBox
   parent=MainInput
   widget=HorizontalBox_0
   slot {class=OverlaySlot, properties {horizontalAlignment=HAlign_Center}}
  }
  Button_Capture {
   widgetClassPath=/FortUILibrary/UIKit/Buttons/WBP_UIKit_Button_Loud
   bIsVariable=true
   parent=HorizontalBox_0
   widget=Button_Capture
   properties {text=Capture, triggeringInputAction {dataTable=&InputActionDataTable, rowName=PhotoModeCaptureConfirm}}
   slot {class=HorizontalBoxSlot, properties {padding=<ltrb 0 0 32 0>}}
  }
  Button_LastSettings {
   widgetClassPath=/FortUILibrary/UIKit/Buttons/WBP_UIKit_Button_Quiet
   bIsVariable=true
   parent=HorizontalBox_0
   widget=Button_LastSettings
   properties {
    bRequiresHold=true
    text=Last Position
    triggeringInputAction {dataTable=&InputActionDataTable, rowName=PhotoModeLastSettings}
   }
   slot {class=HorizontalBoxSlot, properties {padding=<ltrb 0 0 32 0>}}
  }
  Button_Overlay {
   widgetClassPath=&WBP_UIKit_Button_Quiet
   bIsVariable=true
   parent=HorizontalBox_0
   widget=Button_Overlay
   properties {text=Toggle Overlay, triggeringInputAction {dataTable=&InputActionDataTable, rowName=PhotoModeToggleOverlay}}
   slot {class=HorizontalBoxSlot}
  }
  Button_Back {
   widgetClassPath=&WBP_UIKit_Button_Quiet
   bIsVariable=true
   parent=MainInput
   widget=Button_Back
   properties {minHeight=48, text=Back, triggeringInputAction {dataTable=&InputActionDataTable, rowName=PhotoModeBackRetake}}
   slot {class=OverlaySlot, properties {horizontalAlignment=HAlign_Right, verticalAlignment=VAlign_Fill}}
  }
  ConfirmInput {
   widgetClassPath=&HorizontalBox
   parent=Switcher_Input
   widget=ConfirmInput
   slot {class=WidgetSwitcherSlot, properties {horizontalAlignment=HAlign_Center, verticalAlignment=VAlign_Bottom, padding=<ltrb 12 12 12 12>}}
  }
  Button_Confirm {
   widgetClassPath=&WBP_UIKit_Button_Loud
   bIsVariable=true
   parent=ConfirmInput
   widget=Button_Confirm
   properties {text=Confirm, triggeringInputAction {dataTable=&InputActionDataTable, rowName=PhotoModeCaptureConfirm}}
   slot {class=HorizontalBoxSlot, properties {padding=<ltrb 0 0 24 0>}}
  }
  Button_Retake {
   widgetClassPath=&WBP_UIKit_Button_Quiet
   bIsVariable=true
   parent=ConfirmInput
   widget=Button_Retake
   properties {text=Retake, triggeringInputAction {dataTable=&InputActionDataTable, rowName=PhotoModeBackRetake}}
   slot {class=HorizontalBoxSlot}
  }
  CameraSnapEffect {
   widgetClassPath=&Image
   parent=Overlay_0
   widget=CameraSnapEffect
   properties {
    visibility=SelfHitTestInvisible
    brush {
     bIsDynamicallyLoaded=false
     drawAs=Image
     imageType=NoImage
     mirroring=NoMirror
     resourceName=None
     resourceObject=/Game/Athena/HUD/Creative/Materials/M_UI_Creative_PhotoMode_CameraSnapEffect
     tiling=NoTile
     imageSize=<32 32>
     margin=<ltrb 0 0 0 0>
     outlineSettings {
      bUseBrushTransparency=false
      roundingType=HalfHeightRadius
      width=0
      color {colorUseRule=UseColor_Specified, specifiedColor=<rgba 0 0 0 0>}
      cornerRadii=<0 0 0 0>
     }
     tintColor {colorUseRule=UseColor_Specified, specifiedColor=<rgba 0 0 0 1>}
     uVRegion {bIsValid=false, max=<0 0>, min=<0 0>}
    }
   }
   slot {class=OverlaySlot, properties {horizontalAlignment=HAlign_Fill, verticalAlignment=VAlign_Fill}}
  }
 ]
}
