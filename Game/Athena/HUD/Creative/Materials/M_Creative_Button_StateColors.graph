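# Translucent UI material that picks state colors for a Creative HUD button.
# Four StateColors vector parameters (Default, Disabled, Hover, Hover Disabled) are each repacked through MakeFloat4.
# The Disabled scalar blends within the normal and hover pairs, then the Hover scalar blends between the two results.
# A component mask extracts the alpha channel from the final blend.
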Material /Game/Athena/HUD/Creative/Materials/M_Creative_Button_StateColors
graph {
 properties {blendMode=BLEND_Translucent, materialDomain=MD_UI, pixelDepthOffsetMode=PDOM_Legacy, usedShadingModels=Default Lit}
 nodes [
  n1 = VectorParameter() {group=StateColors, parameterName=Default, sortPriority=1, defaultValue=<rgba 0 0 0 1>}
  n2 = MaterialFunctionCall(X=@n1.R, Y=@n1.R, Z=@n1.R, A=@n1.R) {materialFunction=&MakeFloat4}
  n3 = VectorParameter() {group=StateColors, parameterName=Disabled, sortPriority=0, defaultValue=<rgba 0.0885 0.0885 0.0885 0.4>}
  n4 = MaterialFunctionCall(X=@n3.R, Y=@n3.R, Z=@n3.R, A=@n3.R) {materialFunction=&MakeFloat4}
  n5 = ScalarParameter() {parameterName=Disabled, sliderMax=1, sortPriority=0}
  n6 = VectorParameter() {group=StateColors, parameterName=Hover, sortPriority=3, defaultValue=<rgba 1 1 1 1>}
  n7 = MaterialFunctionCall(X=@n6.R, Y=@n6.R, Z=@n6.R, A=@n6.R) {materialFunction=&MakeFloat4}
  n8 = VectorParameter() {group=StateColors, parameterName=Hover Disabled, sortPriority=2, defaultValue=<rgba 0.2159 0 0.1546 0.5>}
  n9 = MaterialFunctionCall(X=@n8.R, Y=@n8.R, Z=@n8.R, A=@n8.R) {materialFunction=&MakeFloat4}
  ComponentMask(None=LinearInterpolate(
    A=LinearInterpolate(A=@n2.Result, B=@n4.Result, Alpha=@n5)
    B=LinearInterpolate(A=@n7.Result, B=@n9.Result, Alpha=@n5)
    Alpha=ScalarParameter() {parameterName=Hover, sortPriority=1}
   )) {a=true}
 ]
}
