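# MF_UI_Creative_RotativeProgress is a UI material function for Creative HUD that outputs a linear progress fill mask along a rotatable axis.
# CustomRotator turns the UVs around a center, which defaults to the middle, so the rotation input sets the fill direction.
# Progress is subtracted from the rotated UVs, smoothstepped between adjustable Min and Max to control edge softness, then the red channel is inverted and saturated.
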MaterialFunction /Game/Athena/HUD/Creative/MaterialFuntions/MF_UI_Creative_RotativeProgress
graph {nodes [
  n1 = FunctionInput() {
   bUsePreviewValueAsDefault=true
   description="Smooth the Progress Alpha gradient between Min and Max  (default value: 0)"
   inputName=Smooth Progress Min
   inputType=FunctionInput_Scalar
   sortPriority=4
  }
  n2 = FunctionInput() {
   bUsePreviewValueAsDefault=true
   description="Smooth the Progress Alpha gradient between Min and Max (default value: 1)"
   inputName=Smooth Progress Max
   inputType=FunctionInput_Scalar
   sortPriority=5
   previewValue=<1 0 0 1>
  }
  n3 = FunctionInput() {
   bUsePreviewValueAsDefault=true
   description="By default in the middle, at 0.5 0.5"
   inputName=Rotation Center
   inputType=FunctionInput_Vector2
   sortPriority=3
   previewValue=<0.5 0.5 0 1>
  }
  n4 = FunctionInput() {bUsePreviewValueAsDefault=true, description=Rotate UVs from 0 to 1, inputName=Rotation, inputType=FunctionInput_Scalar, sortPriority=2}
  n5 = MaterialFunctionCall(
   UVs=FunctionInput(Preview=TextureCoordinate()) {bUsePreviewValueAsDefault=true, inputName=UVs, inputType=FunctionInput_Vector2}
   Rotation Center=@n3
   "Rotation Angle (0-1)"=@n4
  ) {materialFunction=&CustomRotator}
  n6 = FunctionInput() {
   bUsePreviewValueAsDefault=true
   description=fill from 0 to 1
   inputName=Progress
   inputType=FunctionInput_Scalar
   sortPriority=1
   previewValue=<1 0 0 1>
  }
  n7 = ComponentMask(None=SmoothStep(Min=@n1, Max=@n2, Value=Subtract(A=@n5.Rotated Values, B=@n6) {desc="Substract the current \r\nprogress from the UVs"})) {r=true}
  FunctionOutput(None=Saturate(None=OneMinus(None=@n7) {desc=Max Progress is 1})) {outputName=Progress}
 ]}
