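# Translucent Creative HUD material that draws a filled shape and an outline from a channel-packed SDF texture, defaulting to lightning bolt shapes.
# Shape and line each have their own UV transform, lerped or summed RGBA channel selection, and distance-field thresholds.
# Each layer gets opacity and rotatable progress masking through MF_UI_Creative_RotativeProgress.
# The line can hollow out the fill; final color blends line color over a shape color with optional linear or radial gradient.
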Material /Game/Athena/HUD/Creative/Materials/M_UI_Creative_ChannelPack_SDFTexture
graph {
 properties {blendMode=BLEND_Translucent, materialDomain=MD_UI, pixelDepthOffsetMode=PDOM_Legacy, usedShadingModels=Default Lit}
 nodes [
  n1 = NamedRerouteUsage()
  n2 = NamedRerouteDeclaration(None=MaterialFunctionCall!(
    RGBA Channels=Subtract(A=@n1) {constB=0.5}
    Glow Min=ScalarParameter() {group=3. Shape Settings, parameterName=Glow Min, sliderMax=1, sliderMin=-1, sortPriority=9}
    Glow Max=ScalarParameter() {group=3. Shape Settings, parameterName=Glow Max, sliderMax=1, sliderMin=-1, sortPriority=10}
    Outline Thickness
   ) {materialFunction=/FortUILibrary/MaterialFunctions/Basics/MF_UI_ChannelPackedDistanceField}.Fill) {name=Full Shape, nodeColor=<rgba 0 0 0 1>}
  n3 = Multiply(
   A=NamedRerouteUsage()
   B=ScalarParameter() {defaultValue=1, group="2. Color && Opacity", parameterName=Line Hollows Out Shape, sliderMax=1, sortPriority=11}
  )
  NamedRerouteDeclaration(None=Saturate(None=Subtract(A=@n2, B=@n3))) {name=Shape, nodeColor=<rgba 0 0 0 1>}
  n4 = NamedRerouteDeclaration(None=MaterialFunctionCall(
    RGBA Channels=NamedRerouteUsage()
    Glow Min=ScalarParameter() {group=4. Line Settings, parameterName=Line Glow Min, sliderMax=1, sliderMin=-1, sortPriority=9}
    Glow Max=ScalarParameter() {group=4. Line Settings, parameterName=Line Glow Max, sliderMax=1, sliderMin=-1, sortPriority=10}
    Outline Thickness=ScalarParameter() {defaultValue=0.1387, group=4. Line Settings, parameterName=Line Thickness}
   ) {materialFunction=&MF_UI_ChannelPackedDistanceField}.Outline) {name=Full Line, nodeColor=<rgba 0 0 0 1>}
  n5 = OneMinus(None=NamedRerouteUsage())
  NamedRerouteDeclaration(None=Saturate(None=Subtract(A=@n4, B=@n5))) {name=Line, nodeColor=<rgba 0 0 0 1>}
  n6 = NamedRerouteUsage()
  n7 = NamedRerouteUsage()
  NamedRerouteDeclaration(None=Saturate(None=Add(A=@n6, B=@n7))) {name=Final Alpha, nodeColor=<rgba 0 0 0 1>}
  n8 = ScalarParameter() {group="2. Color && Opacity", parameterName=Linear Gradient Rotation, sliderMax=1, sortPriority=3}
  n9 = ScalarParameter() {group="2. Color && Opacity", parameterName=Linear Gradient Offset, sliderMax=1, sliderMin=-1, sortPriority=2}
  n10 = ScalarParameter() {group="2. Color && Opacity", parameterName=Gradient Start, sortPriority=4}
  n11 = ScalarParameter() {defaultValue=1, group="2. Color && Opacity", parameterName=Gradient End, sortPriority=5}
  NamedRerouteDeclaration(None=StaticSwitchParameter(
    True=MaterialFunctionCall(GradientStart=@n10, GradientEnd=@n11, RotationAngle=@n8, GradientOffset=@n9) {materialFunction=&MF_UI_LinearGradient}.Result
    False=MaterialFunctionCall(GradientStart=@n10, GradientEnd=@n11) {materialFunction=&MF_UI_RadialGradient}.Smooth
   ) {defaultValue=true, group="2. Color && Opacity", parameterName=On Linear Gradient / Off Radial Gradient, sortPriority=1}) {name=gradient, nodeColor=<rgba 0 0 0 1>}
  NamedRerouteDeclaration(None=MaterialFunctionCall(
    Input=ScalarParameter() {defaultValue=1, group=4. Line Settings, parameterName=Line Progress, sliderMax=1, sortPriority=0}
    Input Low=Constant()
    Input High=Constant() {r=1}
    Target Low=Constant() {desc="ask adrienne ", r=-1.2}
    Target High=Constant() {r=1}
   ) {materialFunction=&RemapValueRange}.Result) {name=Line Progress Remapped, nodeColor=<rgba 0 0 0 1>}
  n12 = TextureSample(UVs=NamedRerouteUsage(), Tex=NamedRerouteUsage()) {samplerType=SAMPLERTYPE_LinearColor, texture=/Game/Athena/HUD/Creative/EOMScreens/Textures/T_UI_Creative_EOMScreens_LightningBolt_Shapes_1}
  NamedRerouteDeclaration(None=@n12.R) {name=R - Shape, nodeColor=<rgba 0 0 0 1>}
  NamedRerouteDeclaration(None=@n12.G) {name=G - Shape, nodeColor=<rgba 0 0 0 1>}
  NamedRerouteDeclaration(None=@n12.B) {name=B - Shape, nodeColor=<rgba 0 0 0 1>}
  n13 = VectorParameter() {
   desc="Set color, use alpha for global opacity"
   group="2. Color && Opacity"
   parameterName=Line Color
   sortPriority=9
   defaultValue=<rgba 1 0.8 0 1>
  }
  NamedRerouteDeclaration(None=Saturate(None=MaterialFunctionCall(X=@n13.R, Y=@n13.R, Z=@n13.R, A=@n13.R) {materialFunction=&MakeFloat4}.Result)) {name=Line Color, nodeColor=<rgba 0 0 0 1>}
  NamedRerouteDeclaration(None=@n12.A) {name=A - Shape, nodeColor=<rgba 0 0 0 1>}
  n14 = TextureSample(UVs=NamedRerouteUsage(), Tex=NamedRerouteUsage()) {samplerType=SAMPLERTYPE_LinearColor, texture=&T_UI_Creative_EOMScreens_LightningBolt_Shapes_1}
  NamedRerouteDeclaration(None=@n14.R) {name=R - Outline, nodeColor=<rgba 0 0 0 1>}
  n15 = TextureObjectParameter() {
   group=0. Choose your Texture
   parameterName=Texture
   samplerType=SAMPLERTYPE_LinearColor
   sortPriority=0
   texture=&T_UI_Creative_EOMScreens_LightningBolt_Shapes_1
  }
  NamedRerouteDeclaration(None=@n15) {name=Texture, nodeColor=<rgba 0 0 0 1>}
  NamedRerouteDeclaration(None=@n14.G) {name=G - Outline, nodeColor=<rgba 0 0 0 1>}
  NamedRerouteDeclaration(None=@n14.B) {name=B - Outline, nodeColor=<rgba 0 0 0 1>}
  NamedRerouteDeclaration(None=Multiply(A=NamedRerouteUsage(), B=NamedRerouteUsage())) {name=Line With Alpha, nodeColor=<rgba 0 0 0 1>}
  NamedRerouteDeclaration(None=Multiply(A=NamedRerouteUsage(), B=NamedRerouteUsage())) {name=Shape With Alpha, nodeColor=<rgba 0 0 0 1>}
  NamedRerouteDeclaration(None=@n14.A) {name=A - Outline, nodeColor=<rgba 0 0 0 1>}
  n16 = LinearInterpolate(
   A=LinearInterpolate(
    A=NamedRerouteUsage()
    B=NamedRerouteUsage()
    Alpha=ScalarParameter() {group=1. Channel to Display, parameterName=R or G Line, sliderMax=1, sortPriority=9}
   )
   B=NamedRerouteUsage()
   Alpha=ScalarParameter() {group=1. Channel to Display, parameterName=or B Line, sliderMax=1, sortPriority=10}
  )
  n17 = NamedRerouteUsage()
  n18 = ScalarParameter() {group=1. Channel to Display, parameterName=or A Line, sliderMax=1, sortPriority=11}
  n19 = NamedRerouteUsage()
  n20 = Constant()
  n21 = NamedRerouteUsage()
  n22 = Constant()
  n23 = Add(
   A=Add(
    A=StaticSwitchParameter(True=@n19, False=@n20) {defaultValue=true, group=1. Channel to Display, parameterName=R - Line, sortPriority=12}
    B=StaticSwitchParameter(True=@n21, False=@n22) {group=1. Channel to Display, parameterName=G - Line, sortPriority=13}
   )
   B=StaticSwitchParameter(True=NamedRerouteUsage(), False=Constant()) {group=1. Channel to Display, parameterName=B - Line, sortPriority=14}
  )
  n24 = StaticSwitchParameter(True=NamedRerouteUsage(), False=Constant()) {group=1. Channel to Display, parameterName=A - Line, sortPriority=15}
  NamedRerouteDeclaration(None=StaticSwitchParameter(True=LinearInterpolate(A=@n16, B=@n17, Alpha=@n18), False=Add(A=@n23, B=@n24)) {defaultValue=true, group=1. Channel to Display, parameterName="Line - Lerp (true) / Combine (false)", sortPriority=8}) {name=Displayed Lines, nodeColor=<rgba 0 0 0 1>}
  n25 = LinearInterpolate(
   A=LinearInterpolate(
    A=NamedRerouteUsage()
    B=NamedRerouteUsage()
    Alpha=ScalarParameter() {group=1. Channel to Display, parameterName=R or G, sliderMax=1, sortPriority=1}
   )
   B=NamedRerouteUsage()
   Alpha=ScalarParameter() {group=1. Channel to Display, parameterName=or B, sliderMax=1, sortPriority=2}
  )
  n26 = NamedRerouteUsage()
  n27 = ScalarParameter() {group=1. Channel to Display, parameterName=or A, sliderMax=1, sortPriority=3}
  n28 = NamedRerouteUsage()
  n29 = Constant()
  n30 = NamedRerouteUsage()
  n31 = Constant()
  n32 = Add(
   A=Add(
    A=StaticSwitchParameter(True=@n28, False=@n29) {defaultValue=true, group=1. Channel to Display, parameterName=R, sortPriority=4}
    B=StaticSwitchParameter(True=@n30, False=@n31) {group=1. Channel to Display, parameterName=G, sortPriority=5}
   )
   B=StaticSwitchParameter(True=NamedRerouteUsage(), False=Constant()) {group=1. Channel to Display, parameterName=B, sortPriority=6}
  )
  n33 = StaticSwitchParameter(True=NamedRerouteUsage(), False=Constant()) {group=1. Channel to Display, parameterName=A, sortPriority=7}
  NamedRerouteDeclaration(None=StaticSwitchParameter(True=LinearInterpolate(A=@n25, B=@n26, Alpha=@n27), False=Add(A=@n32, B=@n33)) {defaultValue=true, group=1. Channel to Display, parameterName="Lerp (true) / Combine (false)", sortPriority=0}) {name=Displayed Shapes, nodeColor=<rgba 0 0 0 1>}
  NamedRerouteDeclaration(None=MaterialFunctionCall(
    Input=ScalarParameter() {defaultValue=1, group=3. Shape Settings, parameterName=Progress, sliderMax=1, sortPriority=0}
    Input Low=Constant()
    Input High=Constant() {r=1}
    Target Low=Constant() {desc="ask adrienne ", r=-1.2}
    Target High=Constant() {r=1}
   ) {materialFunction=&RemapValueRange}.Result) {name=Progress Remapped, nodeColor=<rgba 0 0 0 1>}
  n34 = MaterialFunctionCall!(
   UVs=NamedRerouteUsage()
   Progress=NamedRerouteUsage()
   Rotation=ScalarParameter() {group=3. Shape Settings, parameterName=Progress Rotation, sliderMax=1, sortPriority=1}
   Rotation Center
   Smooth Progress Min=ScalarParameter() {group=3. Shape Settings, parameterName=Smooth Progress Min, sliderMax=1, sortPriority=2}
   Smooth Progress Max=ScalarParameter() {defaultValue=1, group=3. Shape Settings, parameterName=Smooth Progress Max, sliderMax=1, sortPriority=3}
  ) {
   desc=Compute Progress / Rotation / Alpha Gradient Smoothness
   materialFunction=/Game/Athena/HUD/Creative/MaterialFuntions/MF_UI_Creative_RotativeProgress
  }
  n35 = ScalarParameter() {defaultValue=1, group="2. Color && Opacity", parameterName=Shape Opacity, sliderMax=1, sortPriority=8}
  NamedRerouteDeclaration(None=Multiply(A=@n34.Progress, B=Saturate(None=@n35))) {name=Shape Alpha, nodeColor=<rgba 0 0 0 1>}
  n36 = MaterialFunctionCall(
   UVs=NamedRerouteUsage()
   Progress=NamedRerouteUsage()
   Rotation=ScalarParameter() {group=4. Line Settings, parameterName=Line Progress Rotation, sliderMax=1, sortPriority=1}
   Smooth Progress Min=ScalarParameter() {group=4. Line Settings, parameterName=Line Smooth Progress Min, sliderMax=1, sortPriority=2}
   Smooth Progress Max=ScalarParameter() {defaultValue=1, group=4. Line Settings, parameterName=Line Smooth Progress Max, sliderMax=1, sortPriority=3}
  ) {
   desc=Compute Progress / Rotation / Alpha Gradient Smoothness
   materialFunction=&MF_UI_Creative_RotativeProgress
  }
  n37 = ScalarParameter() {defaultValue=1, group="2. Color && Opacity", parameterName=Line Opacity, sliderMax=1, sortPriority=10}
  NamedRerouteDeclaration(None=Multiply(A=@n36.Progress, B=Saturate(None=@n37))) {name=Line Alpha, nodeColor=<rgba 0 0 0 1>}
  n38 = VectorParameter() {
   desc="Set color, use alpha for global opacity"
   group="2. Color && Opacity"
   parameterName=Shape Color 2
   sortPriority=7
   defaultValue=<rgba 0.6056 1 0.1979 1>
  }
  n39 = VectorParameter() {
   desc="Set color, use alpha for global opacity"
   group="2. Color && Opacity"
   parameterName=Shape Color 1
   sortPriority=6
   defaultValue=<rgba 0.2437 0.534 1 1>
  }
  n40 = MaterialFunctionCall(X=@n39.R, Y=@n39.R, Z=@n39.R, A=@n39.R) {materialFunction=&MakeFloat4}
  n41 = LinearInterpolate(
   A=MaterialFunctionCall(X=@n38.R, Y=@n38.R, Z=@n38.R, A=@n38.R) {materialFunction=&MakeFloat4}.Result
   B=@n40.Result
   Alpha=NamedRerouteUsage()
  )
  NamedRerouteDeclaration(None=StaticSwitchParameter(True=Saturate(None=@n41), False=Saturate(None=@n40.Result)) {group="2. Color && Opacity", parameterName=Use Gradien, sortPriority=0}) {name=Shape Color, nodeColor=<rgba 0 0 0 1>}
  n42 = MaterialFunctionCall(UVs=NamedRerouteUsage(), "Rotation Angle (0-1)"=ScalarParameter() {group=3. Shape Settings, parameterName=Rotation, sortPriority=4}) {materialFunction=&CustomRotator}
  n43 = AppendVector(
   A=ScalarParameter() {defaultValue=1, group=3. Shape Settings, parameterName=Scale X, sortPriority=7}
   B=ScalarParameter() {defaultValue=1, group=3. Shape Settings, parameterName=Scale Y, sortPriority=8}
  )
  n44 = ScalarParameter() {group=3. Shape Settings, parameterName=Offset X, sortPriority=5}
  n45 = ScalarParameter() {group=3. Shape Settings, parameterName=Offset Y, sortPriority=6}
  NamedRerouteDeclaration(None=Add(
    A=MaterialFunctionCall(UVs=@n42.Rotated Values, Texture Scale=@n43) {materialFunction=&ScaleUVsByCenter}.UVs
    B=AppendVector(A=@n44, B=@n45)
   )) {name=Transformed Shape UVs, nodeColor=<rgba 0 0 0 1>}
  NamedRerouteDeclaration(None=MaterialFunctionCall() {materialFunction=&GetUserInterfaceUV}.Normalized UV) {name=Normalized UVs, nodeColor=<rgba 0 0 0 1>}
  n46 = MaterialFunctionCall(UVs=NamedRerouteUsage(), "Rotation Angle (0-1)"=ScalarParameter() {group=4. Line Settings, parameterName=Line Rotation, sortPriority=4}) {materialFunction=&CustomRotator}
  n47 = AppendVector(
   A=ScalarParameter() {defaultValue=1, group=4. Line Settings, parameterName=Line Scale X, sortPriority=7}
   B=ScalarParameter() {defaultValue=1, group=4. Line Settings, parameterName=Line Scale Y, sortPriority=8}
  )
  n48 = ScalarParameter() {group=4. Line Settings, parameterName=Line Offset X, sortPriority=5}
  n49 = ScalarParameter() {group=4. Line Settings, parameterName=Line Offset Y, sortPriority=6}
  NamedRerouteDeclaration(None=Add(
    A=MaterialFunctionCall(UVs=@n46.Rotated Values, Texture Scale=@n47) {materialFunction=&ScaleUVsByCenter}.UVs
    B=AppendVector(A=@n48, B=@n49)
   )) {name=Transformed Line UVs, nodeColor=<rgba 0 0 0 1>}
  NamedRerouteDeclaration(None=LinearInterpolate(A=NamedRerouteUsage(), B=NamedRerouteUsage(), Alpha=NamedRerouteUsage()) {desc="Where the line is display, \r\nwe want the line color"}) {name=Final Color, nodeColor=<rgba 0 0 0 1>}
  NamedRerouteUsage()
  NamedRerouteUsage()
 ]
}
