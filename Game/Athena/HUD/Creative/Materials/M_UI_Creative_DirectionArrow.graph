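# Translucent UI material that procedurally draws a direction arrow for the Creative HUD, with no textures.
# Normalized UI UVs are split into channels, both directly and shifted by Offset X/Y parameters.
# Two hard-edged masks, Arrow and Arrow Dig, are built from UV-style subtract, min/max and Ceil math.
# A Final Arrow subtraction appears to carve one mask from the other, though reroute sources are not shown.
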Material /Game/Athena/HUD/Creative/Materials/M_UI_Creative_DirectionArrow
graph {
 properties {blendMode=BLEND_Translucent, materialDomain=MD_UI, pixelDepthOffsetMode=PDOM_Legacy, usedShadingModels=Default Lit}
 nodes [
  n1 = Saturate(None=Subtract(A=NamedRerouteUsage(), B=NamedRerouteUsage()))
  n2 = NamedRerouteUsage()
  n3 = NamedRerouteUsage()
  n4 = Max(A=@n3, B=@n2)
  n5 = Saturate(None=Subtract(A=Min(A=@n3, B=@n2), B=OneMinus(None=@n4)))
  NamedRerouteDeclaration(None=Min(A=Ceil(None=@n1), B=Ceil(None=@n5))) {name=Arrow Dig, nodeColor=<rgba 0 0 0 1>}
  n6 = NamedRerouteDeclaration(None=MaterialFunctionCall() {materialFunction=&GetUserInterfaceUV}.Normalized UV) {name=Normalized UVs, nodeColor=<rgba 0 0 0 1>}
  NamedRerouteDeclaration(None=ComponentMask(None=@n6) {g=true}) {name=UVs Mask G, nodeColor=<rgba 0 0 0 1>}
  n7 = Add(
   A=NamedRerouteUsage()
   B=AppendVector(A=ScalarParameter() {parameterName=Offset X}, B=ScalarParameter() {defaultValue=-0.36, parameterName=Offset Y})
  )
  NamedRerouteDeclaration(None=ComponentMask(None=@n7) {r=true}) {name=UVs Mask R Offested, nodeColor=<rgba 0 0 0 1>}
  n8 = Saturate(None=Subtract(A=NamedRerouteUsage(), B=NamedRerouteUsage()))
  n9 = NamedRerouteUsage()
  n10 = NamedRerouteUsage()
  n11 = Max(A=@n10, B=@n9)
  n12 = Saturate(None=Subtract(A=Min(A=@n10, B=@n9), B=OneMinus(None=@n11)))
  NamedRerouteDeclaration(None=Min(A=Ceil(None=@n8), B=Ceil(None=@n12))) {name=Arrow, nodeColor=<rgba 0 0 0 1>}
  NamedRerouteDeclaration(None=ComponentMask(None=@n7) {g=true}) {name=UVs Mask G Offested, nodeColor=<rgba 0 0 0 1>}
  NamedRerouteDeclaration(None=Subtract(A=NamedRerouteUsage(), B=NamedRerouteUsage())) {name=Final Arrow, nodeColor=<rgba 0 0 0 1>}
  NamedRerouteDeclaration(None=ComponentMask(None=@n6) {r=true}) {name=UVs Mask R, nodeColor=<rgba 0 0 0 1>}
 ]
}
