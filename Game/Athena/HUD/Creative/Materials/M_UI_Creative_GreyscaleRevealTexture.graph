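# Translucent UI material for Creative HUD that reveals shapes and a background from a single greyscale texture, defaulting to a lightning bolt end-of-match image.
# Shapes Progress minus texel value, amplified, gives shape alpha, so greyscale acts as reveal order.
# Near-white texels form an optional background, wiped by a rotatable, smoothed progress function and added to final alpha.
# Color is solid or a two-color linear or radial gradient via static switches.
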Material /Game/Athena/HUD/Creative/Materials/M_UI_Creative_GreyscaleRevealTexture
graph {
 properties {blendMode=BLEND_Translucent, materialDomain=MD_UI, pixelDepthOffsetMode=PDOM_Legacy, usedShadingModels=Default Lit}
 nodes [
  n1 = Multiply(A=NamedRerouteUsage()) {constB=5}
  n2 = NamedRerouteUsage()
  NamedRerouteDeclaration(None=Saturate(None=Add(A=@n1, B=@n2))) {name=Final Alpha, nodeColor=<rgba 0 0 0 1>}
  n3 = TextureSampleParameter2D(UVs=NamedRerouteUsage()) {
   group=1. Main Settings
   parameterName=Greyscale Texture
   samplerType=SAMPLERTYPE_Alpha
   sortPriority=0
   texture=/Game/Athena/HUD/Creative/EOMScreens/Textures/T_UI_Creative_EOMScreens_LightningBolt_Background_1
  }
  NamedRerouteDeclaration(None=@n3.R) {name=Greyscale Texture, nodeColor=<rgba 0 0 0 1>}
  n4 = Time()
  n5 = Constant() {desc=Speed, r=0.5}
  NamedRerouteDeclaration(None=Frac(None=Multiply(A=@n4, B=@n5))) {name=Animated Progress Test, nodeColor=<rgba 0 0 0 1>}
  n6 = ScalarParameter() {group=2. Color, parameterName=Linear Gradient Rotation, sliderMax=1, sortPriority=4}
  n7 = ScalarParameter() {group=2. Color, parameterName=Linear Gradient Offset, sliderMax=1, sliderMin=-1, sortPriority=7}
  n8 = ScalarParameter() {group=2. Color, parameterName=Gradient Start, sortPriority=5}
  n9 = ScalarParameter() {defaultValue=1, group=2. Color, parameterName=Gradient End, sortPriority=6}
  NamedRerouteDeclaration(None=StaticSwitchParameter(
    True=MaterialFunctionCall(GradientStart=@n8, GradientEnd=@n9, RotationAngle=@n6, GradientOffset=@n7) {materialFunction=&MF_UI_LinearGradient}.Result
    False=MaterialFunctionCall(GradientStart=@n8, GradientEnd=@n9) {materialFunction=&MF_UI_RadialGradient}.Smooth
   ) {defaultValue=true, group=2. Color, parameterName=On Linear Gradient / Off Radial Gradient, sortPriority=1}) {name=gradient, nodeColor=<rgba 0 0 0 1>}
  n10 = VectorParameter() {desc="Set color, use alpha for global opacity", group=2. Color, parameterName=Color 2, sortPriority=3, defaultValue=<rgba 0 0.0285 0.1649 1>}
  n11 = VectorParameter() {desc="Set color, use alpha for global opacity", group=2. Color, parameterName=Color 1, sortPriority=2, defaultValue=<rgba 1 1 1 1>}
  n12 = MaterialFunctionCall(X=@n11.R, Y=@n11.R, Z=@n11.R, A=@n11.R) {materialFunction=&MakeFloat4}
  n13 = LinearInterpolate(
   A=MaterialFunctionCall(X=@n10.R, Y=@n10.R, Z=@n10.R, A=@n10.R) {materialFunction=&MakeFloat4}.Result
   B=@n12.Result
   Alpha=NamedRerouteUsage()
  )
  NamedRerouteDeclaration(None=StaticSwitchParameter(True=Saturate(None=@n13), False=Saturate(None=@n12.Result)) {group=2. Color, parameterName=Use Gradient, sortPriority=0}) {name=Color, nodeColor=<rgba 0 0 0 1>}
  n14 = ScalarParameter() {defaultValue=0.5, group=1. Main Settings, parameterName=Shapes Progress, sliderMax=1, sortPriority=1}
  n15 = NamedRerouteUsage()
  NamedRerouteDeclaration(None=Saturate(None=Subtract(A=@n14, B=@n15))) {name=Shapes Alpha, nodeColor=<rgba 0 0 0 1>}
  n16 = Constant() {desc="1 is the value of the background (beause the background is all the full White on the texture)", r=1}
  n17 = NamedRerouteUsage()
  n18 = OneMinus(None=Multiply(A=Subtract(A=@n16, B=@n17)) {constB=5})
  NamedRerouteDeclaration(None=LinearInterpolate(A=Saturate(None=@n18), Alpha=ScalarParameter() {group=1. Main Settings, parameterName=Use Full Background, sliderMax=1, sortPriority=3})) {name=Background, nodeColor=<rgba 0 0 0 1>}
  n19 = MaterialFunctionCall!(
   UVs=NamedRerouteUsage()
   Progress=NamedRerouteUsage()
   Rotation=ScalarParameter() {group=1. Main Settings, parameterName=Background Rotation, sliderMax=1, sortPriority=5}
   Rotation Center
   Smooth Progress Min=ScalarParameter() {group=1. Main Settings, parameterName=Background Smooth Progress Min, sliderMax=1, sortPriority=6}
   Smooth Progress Max=ScalarParameter() {defaultValue=1, group=1. Main Settings, parameterName=Background Smooth Progress Max, sliderMax=1, sortPriority=7}
  ) {
   desc=Compute Progress / Rotation / Alpha Gradient Smoothness
   materialFunction=/Game/Athena/HUD/Creative/MaterialFuntions/MF_UI_Creative_RotativeProgress
  }
  NamedRerouteDeclaration(None=Multiply(A=NamedRerouteUsage(), B=@n19.Progress)) {name=Background Alpha, nodeColor=<rgba 0 0 0 1>}
  NamedRerouteDeclaration(None=MaterialFunctionCall() {materialFunction=&GetUserInterfaceUV}.Normalized UV) {name=Normalized UVs, nodeColor=<rgba 0 0 0 1>}
  NamedRerouteDeclaration(None=MaterialFunctionCall(
    Input=ScalarParameter() {defaultValue=0.5, group=1. Main Settings, parameterName=Background Progress, sliderMax=1, sortPriority=2}
    Input Low=Constant()
    Input High=Constant() {r=1}
    Target Low=Constant() {desc="ask adrienne ", r=-1.2}
    Target High=Constant() {r=1}
   ) {materialFunction=&RemapValueRange}.Result) {name=Background Progress Remapped, nodeColor=<rgba 0 0 0 1>}
  NamedRerouteUsage()
  NamedRerouteUsage()
  NamedRerouteUsage()
  NamedRerouteUsage()
 ]
}
