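# Translucent UI material for the Creative HUD that draws a configurable triangle with a fill and an outline.
# MF_UI_WIP_Triangle builds fill and outline separately, each with its own vertex, offset, scale and glow parameters.
# Remapped progress values drive rotatable, smoothed wipes that combine with opacity to mask each layer.
# Named reroutes composite the line color with a shape color, which static switches make solid or a linear or radial gradient.
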Material /Game/Athena/HUD/Creative/Materials/M_UI_Creative_Polygon_3_Vertices
graph {
 properties {blendMode=BLEND_Translucent, materialDomain=MD_UI, pixelDepthOffsetMode=PDOM_Legacy, usedShadingModels=Default Lit}
 nodes [
  n1 = MaterialFunctionCall!(
   UVs=NamedRerouteUsage()
   Corner 01=NamedRerouteUsage()
   Corner 02=NamedRerouteUsage()
   Corner 03=NamedRerouteUsage()
   Glow Min=ScalarParameter() {group=Shape, parameterName=Shape Glow Min, sortPriority=7}
   Glow Max=ScalarParameter() {group=Shape, parameterName=Shape Glow Max, sortPriority=8}
   Outline Thickness
  ) {materialFunction=/FortUILibrary/MaterialFunctions/Shapes/MF_UI_WIP_Triangle}
  n2 = NamedRerouteUsage()
  NamedRerouteDeclaration(None=Saturate(None=Subtract(A=@n1.Fill, B=@n2))) {name=Shape, nodeColor=<rgba 0 0 0 1>}
  n3 = VectorParameter() {
   desc="Set color, use alpha for global opacity"
   group="_Color && Opacity"
   parameterName=Shape Color 1
   sortPriority=1
   defaultValue=<rgba 0.1758 1 0.1597 1>
  }
  n4 = VectorParameter() {
   desc="Set color, use alpha for global opacity"
   group="_Color && Opacity"
   parameterName=Shape Color 2
   sortPriority=2
   defaultValue=<rgba 0.159 0.4845 1 1>
  }
  n5 = MaterialFunctionCall(X=@n4.R, Y=@n4.R, Z=@n4.R, A=@n4.R) {materialFunction=&MakeFloat4}
  n6 = LinearInterpolate(
   A=MaterialFunctionCall(X=@n3.R, Y=@n3.R, Z=@n3.R, A=@n3.R) {materialFunction=&MakeFloat4}.Result
   B=@n5.Result
   Alpha=NamedRerouteUsage()
  )
  NamedRerouteDeclaration(None=StaticSwitchParameter(True=Saturate(None=@n6), False=Saturate(None=@n5.Result)) {group="_Color && Opacity", parameterName=Use Gradien, sortPriority=0}) {name=Shape Color, nodeColor=<rgba 0 0 0 1>}
  NamedRerouteDeclaration(None=LinearInterpolate(A=NamedRerouteUsage(), B=NamedRerouteUsage(), Alpha=NamedRerouteUsage()) {desc="Where the line is display, \r\nwe want the line color"}) {name=Final Color, nodeColor=<rgba 0 0 0 1>}
  NamedRerouteDeclaration(None=AppendVector(
    A=ScalarParameter() {group=Shape, parameterName=Shape Offset X, sortPriority=5}
    B=ScalarParameter() {group=Shape, parameterName=Shape Offset Y, sortPriority=6}
   )) {name=Shape Offset, nodeColor=<rgba 0 0 0 1>}
  NamedRerouteDeclaration(None=MaterialFunctionCall(UVs=NamedRerouteUsage(), Texture Scale=ScalarParameter() {defaultValue=1, group=Shape, parameterName=Shape Scale Factor, sortPriority=4}) {materialFunction=&ScaleUVsByCenter}.UVs) {name=Shape UVs Scale, nodeColor=<rgba 0 0 0 1>}
  n7 = VectorParameter() {
   desc="Set color, use alpha for global opacity"
   group="_Color && Opacity"
   parameterName=Line Color
   sortPriority=0
   defaultValue=<rgba 1 0.8485 0 1>
  }
  NamedRerouteDeclaration(None=Saturate(None=MaterialFunctionCall(X=@n7.R, Y=@n7.R, Z=@n7.R, A=@n7.R) {materialFunction=&MakeFloat4}.Result)) {name=Line Color, nodeColor=<rgba 0 0 0 1>}
  n8 = ScalarParameter() {defaultValue=1, group=Shape Vertex Coordinates, parameterName=Shape Vertex 1 X, sliderMax=2, sliderMin=-2, sortPriority=34}
  n9 = ScalarParameter() {defaultValue=1, group=Shape Vertex Coordinates, parameterName=Shape Vertex 1 Y, sliderMax=2, sliderMin=-2, sortPriority=35}
  NamedRerouteDeclaration(None=Add(A=AppendVector(A=@n8, B=@n9), B=NamedRerouteUsage())) {name=Shape Vertex 1, nodeColor=<rgba 0 0 0 1>}
  n10 = MaterialFunctionCall!(
   UVs=NamedRerouteUsage()
   Progress=NamedRerouteUsage()
   Rotation=ScalarParameter() {group=Line, parameterName=Line Progress Rotation, sliderMax=1, sortPriority=1}
   Rotation Center
   Smooth Progress Min=ScalarParameter() {group=Line, parameterName=Line Smooth Progress Min, sliderMax=1, sortPriority=2}
   Smooth Progress Max=ScalarParameter() {defaultValue=1, group=Line, parameterName=Line Smooth Progress Max, sliderMax=1, sortPriority=3}
  ) {
   desc=Compute Progress / Rotation / Alpha Gradient Smoothness
   materialFunction=/Game/Athena/HUD/Creative/MaterialFuntions/MF_UI_Creative_RotativeProgress
  }
  n11 = ScalarParameter() {defaultValue=1, group="_Color && Opacity", parameterName=Line Opacity, sliderMax=1, sortPriority=8}
  NamedRerouteDeclaration(None=Multiply(A=@n10.Progress, B=Saturate(None=@n11))) {name=Line Alpha, nodeColor=<rgba 0 0 0 1>}
  n12 = NamedRerouteDeclaration(None=MaterialFunctionCall(
    UVs=NamedRerouteUsage()
    Corner 01=NamedRerouteUsage()
    Corner 02=NamedRerouteUsage()
    Corner 03=NamedRerouteUsage()
    Glow Min=ScalarParameter() {group=Line, parameterName=Line Glow Min, sortPriority=7}
    Glow Max=ScalarParameter() {group=Line, parameterName=Line Glow Max, sortPriority=8}
    Outline Thickness=ScalarParameter() {defaultValue=0.015, group=Line, parameterName=Line Thickness, sortPriority=-1}
   ) {materialFunction=&MF_UI_WIP_Triangle}.Outline) {name=Full Line, nodeColor=<rgba 0 0 0 1>}
  n13 = OneMinus(None=NamedRerouteUsage())
  NamedRerouteDeclaration(None=Saturate(None=Subtract(A=@n12, B=@n13))) {name=Line, nodeColor=<rgba 0 0 0 1>}
  NamedRerouteDeclaration(None=@n1.Fill) {name=Full Shape, nodeColor=<rgba 0 0 0 1>}
  NamedRerouteDeclaration(None=Multiply(A=NamedRerouteUsage(), B=NamedRerouteUsage())) {name=Line With Alpha, nodeColor=<rgba 0 0 0 1>}
  NamedRerouteDeclaration(None=Multiply(A=NamedRerouteUsage(), B=NamedRerouteUsage())) {name=Shape With Alpha, nodeColor=<rgba 0 0 0 1>}
  n14 = ScalarParameter() {group="_Color && Opacity", parameterName=Linear Gradient Rotation, sliderMax=1, sortPriority=4}
  n15 = ScalarParameter() {group="_Color && Opacity", parameterName=Linear Gradient Offset, sliderMax=1, sliderMin=-1, sortPriority=7}
  n16 = ScalarParameter() {group="_Color && Opacity", parameterName=Gradient Start, sortPriority=5}
  n17 = ScalarParameter() {defaultValue=1, group="_Color && Opacity", parameterName=Gradient End, sortPriority=6}
  NamedRerouteDeclaration(None=StaticSwitchParameter(
    True=MaterialFunctionCall(GradientStart=@n16, GradientEnd=@n17, RotationAngle=@n14, GradientOffset=@n15) {materialFunction=&MF_UI_LinearGradient}.Result
    False=MaterialFunctionCall(GradientStart=@n16, GradientEnd=@n17) {materialFunction=&MF_UI_RadialGradient}.Smooth
   ) {defaultValue=true, group="_Color && Opacity", parameterName=On Linear Gradient / Off Radial Gradient, sortPriority=1}) {name=gradient, nodeColor=<rgba 0 0 0 1>}
  n18 = ScalarParameter() {defaultValue=-1, group=Shape Vertex Coordinates, parameterName=Shape Vertex 2 X, sliderMax=2, sliderMin=-2, sortPriority=36}
  n19 = ScalarParameter() {defaultValue=1, group=Shape Vertex Coordinates, parameterName=Shape Vertex 2 Y, sliderMax=2, sliderMin=-2, sortPriority=37}
  NamedRerouteDeclaration(None=Add(A=AppendVector(A=@n18, B=@n19), B=NamedRerouteUsage())) {name=Shape Vertex 2, nodeColor=<rgba 0 0 0 1>}
  n20 = NamedRerouteUsage()
  n21 = NamedRerouteUsage()
  NamedRerouteDeclaration(None=Saturate(None=Add(A=@n20, B=@n21))) {name=Final Alpha, nodeColor=<rgba 0 0 0 1>}
  NamedRerouteDeclaration(None=AppendVector(
    A=ScalarParameter() {group=Line, parameterName=Line Offset X, sortPriority=5}
    B=ScalarParameter() {group=Line, parameterName=Line Offset Y, sortPriority=6}
   )) {name=Line Offset, nodeColor=<rgba 0 0 0 1>}
  NamedRerouteDeclaration(None=MaterialFunctionCall(UVs=NamedRerouteUsage(), Texture Scale=ScalarParameter() {defaultValue=0.6, group=Line, parameterName=Line Scale Factor, sortPriority=4}) {materialFunction=&ScaleUVsByCenter}.UVs) {name=Line UVs Scale, nodeColor=<rgba 0 0 0 1>}
  n22 = ScalarParameter() {group=Line Vertex Coordinates, parameterName=Line Vertex 0 X, sliderMax=2, sliderMin=-2}
  n23 = ScalarParameter() {defaultValue=-1, group=Line Vertex Coordinates, parameterName=Line Vertex 0 Y, sliderMax=2, sliderMin=-2, sortPriority=33}
  NamedRerouteDeclaration(None=Add(A=AppendVector(A=@n22, B=@n23), B=NamedRerouteUsage())) {name=Line Vertex 0, nodeColor=<rgba 0 0 0 1>}
  n24 = ScalarParameter() {defaultValue=1, group=Line Vertex Coordinates, parameterName=Line Vertex 1 X, sliderMax=2, sliderMin=-2, sortPriority=34}
  n25 = ScalarParameter() {defaultValue=1, group=Line Vertex Coordinates, parameterName=Line Vertex 1 Y, sliderMax=2, sliderMin=-2, sortPriority=35}
  NamedRerouteDeclaration(None=Add(A=AppendVector(A=@n24, B=@n25), B=NamedRerouteUsage())) {name=Line Vertex 1, nodeColor=<rgba 0 0 0 1>}
  n26 = ScalarParameter() {defaultValue=-1, group=Line Vertex Coordinates, parameterName=Line Vertex 2 X, sliderMax=2, sliderMin=-2, sortPriority=36}
  n27 = ScalarParameter() {defaultValue=1, group=Line Vertex Coordinates, parameterName=Line Vertex 2 Y, sliderMax=2, sliderMin=-2, sortPriority=37}
  NamedRerouteDeclaration(None=Add(A=AppendVector(A=@n26, B=@n27), B=NamedRerouteUsage())) {name=Line Vertex 2, nodeColor=<rgba 0 0 0 1>}
  NamedRerouteDeclaration(None=MaterialFunctionCall(
    Input=ScalarParameter() {defaultValue=1, group=Line, parameterName=Line Progress, sliderMax=1, sortPriority=0}
    Input Low=Constant()
    Input High=Constant() {r=1}
    Target Low=Constant() {desc="ask adrienne ", r=-1.2}
    Target High=Constant() {r=1}
   ) {materialFunction=&RemapValueRange}.Result) {name=Line Progress Remapped, nodeColor=<rgba 0 0 0 1>}
  n28 = MaterialFunctionCall(
   UVs=NamedRerouteUsage()
   Progress=NamedRerouteUsage()
   Rotation=ScalarParameter() {group=Shape, parameterName=Shape Progress Rotation, sliderMax=1, sortPriority=1}
   Smooth Progress Min=ScalarParameter() {group=Shape, parameterName=Shape Smooth Progress Min, sliderMax=1, sortPriority=2}
   Smooth Progress Max=ScalarParameter() {defaultValue=1, group=Shape, parameterName=Shape Smooth Progress Max, sliderMax=1, sortPriority=3}
  ) {
   desc=Compute Progress / Rotation / Alpha Gradient Smoothness
   materialFunction=&MF_UI_Creative_RotativeProgress
  }
  n29 = ScalarParameter() {defaultValue=1, group="_Color && Opacity", parameterName=Shape Opacity, sliderMax=1, sortPriority=9}
  NamedRerouteDeclaration(None=Multiply(A=@n28.Progress, B=Saturate(None=@n29))) {name=Shape Alpha, nodeColor=<rgba 0 0 0 1>}
  NamedRerouteDeclaration(None=MaterialFunctionCall() {materialFunction=&GetUserInterfaceUV}.Normalized UV) {name=Normalized UVs, nodeColor=<rgba 0 0 0 1>}
  n30 = ScalarParameter() {group=Shape Vertex Coordinates, parameterName=Shape Vertex 0 X, sliderMax=2, sliderMin=-2}
  n31 = ScalarParameter() {defaultValue=-1, group=Shape Vertex Coordinates, parameterName=Shape Vertex 0 Y, sliderMax=2, sliderMin=-2, sortPriority=33}
  NamedRerouteDeclaration(None=Add(A=AppendVector(A=@n30, B=@n31), B=NamedRerouteUsage())) {name=Shape Vertex 0, nodeColor=<rgba 0 0 0 1>}
  NamedRerouteDeclaration(None=MaterialFunctionCall(
    Input=ScalarParameter() {defaultValue=1, group=Shape, parameterName=Shape Progress, sliderMax=1, sortPriority=0}
    Input Low=Constant()
    Input High=Constant() {r=1}
    Target Low=Constant() {desc="ask adrienne ", r=-1.2}
    Target High=Constant() {r=1}
   ) {materialFunction=&RemapValueRange}.Result) {name=Shape Progress Remapped, nodeColor=<rgba 0 0 0 1>}
  NamedRerouteUsage()
  NamedRerouteUsage()
 ]
}
