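# Translucent UI material for Creative HUD that draws three editable four-vertex polygons over a colored background.
# Each quad's shape comes from a polygon function fed by corner parameters, with offset, scale and glow controls.
# Background and quads each get a color and an alpha from opacity times a remapped, rotatable progress gradient.
# Colored layers and alphas are summed, and a static switch chooses how the background shows relative to the shapes.
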Material /Game/Athena/HUD/Creative/Materials/M_UI_Creative_Polygon_4_Vertices_X3
graph {
 properties {blendMode=BLEND_Translucent, materialDomain=MD_UI, pixelDepthOffsetMode=PDOM_Legacy, usedShadingModels=Default Lit}
 nodes [
  Composite() {subgraphName=V1 Info}
  Composite() {subgraphName=V2 Info}
  n1 = ScalarParameter() {defaultValue=-1, group=Quad 1 - Vertex Coordinates, parameterName=Quad 1 - TopLeft Y, sliderMax=2, sliderMin=-2, sortPriority=33}
  n2 = AppendVector(A=ScalarParameter() {defaultValue=-1, group=Quad 1 - Vertex Coordinates, parameterName=Quad 1 - TopLeft X, sliderMax=2, sliderMin=-2}, B=@n1)
  n3 = NamedRerouteUsage()
  n4 = ScalarParameter() {defaultValue=1, group=Quad 1 - Vertex Coordinates, parameterName=Quad 1 - TopRight X, sliderMax=2, sliderMin=-2, sortPriority=34}
  n5 = ScalarParameter() {defaultValue=-1, group=Quad 1 - Vertex Coordinates, parameterName=Quad 1 - TopRight Y, sliderMax=2, sliderMin=-2, sortPriority=35}
  n6 = AppendVector(A=@n4, B=@n5)
  n7 = NamedRerouteUsage()
  n8 = ScalarParameter() {defaultValue=1, group=Quad 1 - Vertex Coordinates, parameterName=Quad 1 - BottomRight X, sliderMax=2, sliderMin=-2, sortPriority=36}
  n9 = ScalarParameter() {defaultValue=1, group=Quad 1 - Vertex Coordinates, parameterName=Quad 1 - BottomRight Y, sliderMax=2, sliderMin=-2, sortPriority=37}
  n10 = AppendVector(A=@n8, B=@n9)
  n11 = NamedRerouteUsage()
  n12 = ScalarParameter() {defaultValue=-1, group=Quad 1 - Vertex Coordinates, parameterName=Quad 1 - BottomLeft X, sliderMax=2, sliderMin=-2, sortPriority=38}
  n13 = ScalarParameter() {defaultValue=1, group=Quad 1 - Vertex Coordinates, parameterName=Quad 1 - BottomLeft Y, sliderMax=2, sliderMin=-2, sortPriority=39}
  n14 = AppendVector(A=@n12, B=@n13)
  n15 = NamedRerouteUsage()
  NamedRerouteDeclaration(None=MaterialFunctionCall!(
    UVs=NamedRerouteUsage()
    vertex_0=Add(A=@n2, B=@n3)
    vertex_1=Add(A=@n6, B=@n7)
    vertex_2=Add(A=@n10, B=@n11)
    vertex_3=Add(A=@n14, B=@n15)
    Outline Thickness
    Glow Min=ScalarParameter() {group=Quad 1, parameterName=Quad 1 Glow Min, sortPriority=7}
    Glow Max=ScalarParameter() {group=Quad 1, parameterName=Quad 1 Glow Max, sortPriority=8}
   ) {materialFunction=/Game/Athena/HUD/Creative/MaterialFuntions/MF_UI_WIP_Creative_Polygon_4_Vertex}.Fill) {name=Quad 1 Shape, nodeColor=<rgba 0 0 0 1>}
  n16 = VectorParameter() {
   desc="Set color, use alpha for global opacity"
   group="0. Color && Opacity"
   parameterName=Quad 2 Color
   sortPriority=2
   defaultValue=<rgba 1 0.0669 0.3949 1>
  }
  NamedRerouteDeclaration(None=Saturate(None=@n16.RGB)) {name=Quad 2 Color, nodeColor=<rgba 0 0 0 1>}
  n17 = MaterialFunctionCall!(
   UVs=NamedRerouteUsage()
   Progress=NamedRerouteUsage()
   Rotation=ScalarParameter() {group=Quad 3, parameterName=Quad 3 Progress Rotation, sliderMax=1, sortPriority=1}
   Rotation Center
   Smooth Progress Min=ScalarParameter() {group=Quad 3, parameterName=Quad 3 Smooth Progress Min, sliderMax=1, sortPriority=2}
   Smooth Progress Max=ScalarParameter() {defaultValue=1, group=Quad 3, parameterName=Quad 3 Smooth Progress Max, sliderMax=1, sortPriority=3}
  ) {
   desc=Compute Progress / Rotation / Alpha Gradient Smoothness
   materialFunction=/Game/Athena/HUD/Creative/MaterialFuntions/MF_UI_Creative_RotativeProgress
  }
  n18 = ScalarParameter() {defaultValue=0.5, group="0. Color && Opacity", parameterName=Quad 3 Opacity, sliderMax=1, sortPriority=7}
  NamedRerouteDeclaration(None=Multiply(A=@n17.Progress, B=Saturate(None=@n18))) {name=Quad 3 Alpha, nodeColor=<rgba 0 0 0 1>}
  NamedRerouteDeclaration(None=MaterialFunctionCall(
    Input=ScalarParameter() {defaultValue=1, group=Quad 1, parameterName=Quad 1 Progress, sliderMax=1, sortPriority=0}
    Input Low=NamedRerouteUsage()
    Input High=NamedRerouteUsage()
    Target Low=NamedRerouteUsage()
    Target High=NamedRerouteUsage()
   ) {materialFunction=&RemapValueRange}.Result) {name=Quad 1 Progress Remapped, nodeColor=<rgba 0 0 0 1>}
  n19 = MaterialFunctionCall(
   UVs=NamedRerouteUsage()
   Progress=NamedRerouteUsage()
   Rotation=ScalarParameter() {group=Background, parameterName=Bg Progress Rotation, sliderMax=1, sortPriority=1}
   Smooth Progress Min=ScalarParameter() {group=Background, parameterName=Bg Smooth Progress Min, sliderMax=1, sortPriority=2}
   Smooth Progress Max=ScalarParameter() {defaultValue=1, group=Background, parameterName=Bg Smooth Progress Max, sliderMax=1, sortPriority=3}
  ) {
   desc=Compute Progress / Rotation / Alpha Gradient Smoothness
   materialFunction=&MF_UI_Creative_RotativeProgress
  }
  n20 = ScalarParameter() {defaultValue=1, group="0. Color && Opacity", parameterName=Bg Opacity, sliderMax=1, sortPriority=4}
  NamedRerouteDeclaration(None=Multiply(A=@n19.Progress, B=Saturate(None=@n20))) {name=Background Alpha, nodeColor=<rgba 0 0 0 1>}
  n21 = NamedRerouteUsage()
  n22 = NamedRerouteUsage()
  n23 = NamedRerouteUsage()
  n24 = NamedRerouteUsage()
  n25 = Add(
   A=NamedRerouteDeclaration(None=Multiply(A=@n21, B=@n22)) {name=Colored Background, nodeColor=<rgba 0 0 0 1>}
   B=NamedRerouteDeclaration(None=Multiply(A=@n23, B=@n24)) {name=Colored Quad 1, nodeColor=<rgba 0 0 0 1>}
  )
  n26 = NamedRerouteDeclaration(None=Multiply(A=NamedRerouteUsage(), B=NamedRerouteUsage())) {name=Colored Quad 2, nodeColor=<rgba 0 0 0 1>}
  n27 = Multiply(A=NamedRerouteUsage(), B=NamedRerouteUsage())
  NamedRerouteDeclaration(None=Add(A=Add(A=@n25, B=@n26), B=NamedRerouteDeclaration(None=@n27) {name=Colored Quad 3, nodeColor=<rgba 0 0 0 1>})) {name=Colored Shapes, nodeColor=<rgba 0 0 0 1>}
  NamedRerouteDeclaration(None=AppendVector(
    A=ScalarParameter() {defaultValue=-0.9, group=Quad 1, parameterName=Quad 1 Offset X, sortPriority=5}
    B=ScalarParameter() {defaultValue=-0.9, group=Quad 1, parameterName=Quad 1 Offset Y, sortPriority=6}
   )) {name=Quad 1 Offset, nodeColor=<rgba 0 0 0 1>}
  NamedRerouteDeclaration(None=StaticSwitchParameter(True=NamedRerouteUsage(), False=NamedRerouteUsage()) {defaultValue=true, group="0. Color && Opacity", parameterName=Fill Shape Transparencies With Bg Color, sortPriority=-1}) {name=Visible Background, nodeColor=<rgba 0 0 0 1>}
  NamedRerouteDeclaration(None=MaterialFunctionCall(Texture Scale=ScalarParameter() {defaultValue=0.5, group=Quad 1, parameterName=Quad 1 Scale Factor, sortPriority=4}) {materialFunction=&ScaleUVsByCenter}.UVs) {name=Quad 1 UVs Scale, nodeColor=<rgba 0 0 0 1>}
  n28 = VectorParameter() {
   desc="Set color, use alpha for global opacity"
   group="0. Color && Opacity"
   parameterName=Bg Color
   sortPriority=0
   defaultValue=<rgba 0 0.0285 0.1649 1>
  }
  NamedRerouteDeclaration(None=Saturate(None=@n28.RGB)) {name=Background Color, nodeColor=<rgba 0 0 0 1>}
  n29 = VectorParameter() {
   desc="Set color, use alpha for global opacity"
   group="0. Color && Opacity"
   parameterName=Quad 1 Color
   sortPriority=1
   defaultValue=<rgba 0.1758 1 0.1597 1>
  }
  NamedRerouteDeclaration(None=Saturate(None=@n29.RGB)) {name=Quad 1 Color, nodeColor=<rgba 0 0 0 1>}
  NamedRerouteDeclaration(None=MaterialFunctionCall(
    Input=ScalarParameter() {defaultValue=1, group=Background, parameterName=Bg Progress, sliderMax=1, sortPriority=0}
    Input Low=NamedRerouteUsage()
    Input High=NamedRerouteUsage()
    Target Low=NamedRerouteUsage()
    Target High=NamedRerouteUsage()
   ) {materialFunction=&RemapValueRange}.Result) {name=Bg Progress Remapped, nodeColor=<rgba 0 0 0 1>}
  n30 = VectorParameter() {
   desc="Set color, use alpha for global opacity"
   group="0. Color && Opacity"
   parameterName=Quad 3 Color
   sortPriority=3
   defaultValue=<rgba 1 0.8548 0 1>
  }
  NamedRerouteDeclaration(None=Saturate(None=@n30.RGB)) {name=Quad 3 Color, nodeColor=<rgba 0 0 0 1>}
  NamedRerouteDeclaration(None=AppendVector(
    A=ScalarParameter() {group=Quad 2, parameterName=Quad 2 Offset X, sortPriority=5}
    B=ScalarParameter() {group=Quad 2, parameterName=Quad 2 Offset Y, sortPriority=6}
   )) {name=Quad 2 Offset, nodeColor=<rgba 0 0 0 1>}
  NamedRerouteDeclaration(None=MaterialFunctionCall(Texture Scale=ScalarParameter() {defaultValue=0.5, group=Quad 2, parameterName=Quad 2 Scale Factor, sortPriority=4}) {materialFunction=&ScaleUVsByCenter}.UVs) {name=Quad 2 UVs Scale, nodeColor=<rgba 0 0 0 1>}
  NamedRerouteDeclaration(None=Multiply(A=NamedRerouteUsage(), B=NamedRerouteUsage())) {name=Background Shape With Alpha, nodeColor=<rgba 0 0 0 1>}
  NamedRerouteDeclaration(None=Multiply(A=NamedRerouteUsage(), B=NamedRerouteUsage())) {name=Quad 1 Shape With Alpha, nodeColor=<rgba 0 0 0 1>}
  NamedRerouteDeclaration(None=AppendVector(
    A=ScalarParameter() {defaultValue=0.9, group=Quad 3, parameterName=Quad 3 Offset X, sortPriority=5}
    B=ScalarParameter() {defaultValue=0.9, group=Quad 3, parameterName=Quad 3 Offset Y, sortPriority=6}
   )) {name=Quad 3 Offset, nodeColor=<rgba 0 0 0 1>}
  NamedRerouteDeclaration(None=MaterialFunctionCall(Texture Scale=ScalarParameter() {defaultValue=0.5, group=Quad 3, parameterName="Quad \" Scale Factor", sortPriority=4}) {materialFunction=&ScaleUVsByCenter}.UVs) {name=Quad 3 UVs Scale, nodeColor=<rgba 0 0 0 1>}
  NamedRerouteDeclaration(None=Multiply(A=NamedRerouteUsage(), B=NamedRerouteUsage())) {name=Quad 2 Shape With Alpha, nodeColor=<rgba 0 0 0 1>}
  NamedRerouteDeclaration(None=Multiply(A=NamedRerouteUsage(), B=NamedRerouteUsage())) {name=Quad 3 Shape With Alpha, nodeColor=<rgba 0 0 0 1>}
  NamedRerouteDeclaration(None=Constant() {r=1}) {name=Background Shape, nodeColor=<rgba 0 0 0 1>}
  NamedRerouteDeclaration(None=MaterialFunctionCall(
    Input=ScalarParameter() {defaultValue=1, group=Quad 2, parameterName=Quad 2 Progress, sliderMax=1, sortPriority=0}
    Input Low=NamedRerouteUsage()
    Input High=NamedRerouteUsage()
    Target Low=NamedRerouteUsage()
    Target High=NamedRerouteUsage()
   ) {materialFunction=&RemapValueRange}.Result) {name=Quad 2 Progress Remapped, nodeColor=<rgba 0 0 0 1>}
  NamedRerouteDeclaration(None=MaterialFunctionCall(
    Input=ScalarParameter() {defaultValue=1, group=Quad 3, parameterName=Quad 3 Progress, sliderMax=1, sortPriority=0}
    Input Low=NamedRerouteUsage()
    Input High=NamedRerouteUsage()
    Target Low=NamedRerouteUsage()
    Target High=NamedRerouteUsage()
   ) {materialFunction=&RemapValueRange}.Result) {name=Quad 3 Progress Remapped, nodeColor=<rgba 0 0 0 1>}
  NamedRerouteDeclaration(None=Constant()) {name=Input Low, nodeColor=<rgba 0 0 0 1>}
  NamedRerouteDeclaration(None=Constant() {r=1}) {name=Input High, nodeColor=<rgba 0 0 0 1>}
  n31 = NamedRerouteUsage()
  n32 = Add(A=Add(A=NamedRerouteUsage(), B=NamedRerouteUsage()), B=NamedRerouteUsage())
  NamedRerouteDeclaration(None=Saturate(None=Subtract(A=@n31, B=@n32))) {name=Visible Bg Outside Shapes, nodeColor=<rgba 0 0 0 1>}
  n33 = NamedRerouteUsage()
  n34 = Add(A=Add(A=NamedRerouteUsage(), B=NamedRerouteUsage()), B=NamedRerouteUsage())
  NamedRerouteDeclaration(None=Saturate(None=Subtract(A=@n33, B=@n34))) {name=Visible Bg Inside Shapes, nodeColor=<rgba 0 0 0 1>}
  NamedRerouteDeclaration(None=Constant() {desc="ask adrienne ", r=-1.2}) {name=Target Low, nodeColor=<rgba 0 0 0 1>}
  NamedRerouteDeclaration(None=Constant() {r=1}) {name=Target High, nodeColor=<rgba 0 0 0 1>}
  n35 = ScalarParameter() {defaultValue=-1, group=Quad 2 Vertex Coordinates, parameterName=Quad 2 - TopLeft Y, sliderMax=2, sliderMin=-2, sortPriority=33}
  n36 = AppendVector(A=ScalarParameter() {defaultValue=-1, group=Quad 2 Vertex Coordinates, parameterName=Quad 2 - TopLeft X, sliderMax=2, sliderMin=-2}, B=@n35)
  n37 = NamedRerouteUsage()
  n38 = ScalarParameter() {defaultValue=1, group=Quad 2 Vertex Coordinates, parameterName=Quad 2 - TopRight X, sliderMax=2, sliderMin=-2, sortPriority=34}
  n39 = ScalarParameter() {defaultValue=-1, group=Quad 2 Vertex Coordinates, parameterName=Quad 2 - TopRight Y, sliderMax=2, sliderMin=-2, sortPriority=35}
  n40 = AppendVector(A=@n38, B=@n39)
  n41 = NamedRerouteUsage()
  n42 = ScalarParameter() {defaultValue=1, group=Quad 2 Vertex Coordinates, parameterName=Quad 2 - BottomRight X, sliderMax=2, sliderMin=-2, sortPriority=36}
  n43 = ScalarParameter() {defaultValue=1, group=Quad 2 Vertex Coordinates, parameterName=Quad 2 - BottomRight Y, sliderMax=2, sliderMin=-2, sortPriority=37}
  n44 = AppendVector(A=@n42, B=@n43)
  n45 = NamedRerouteUsage()
  n46 = ScalarParameter() {defaultValue=-1, group=Quad 2 Vertex Coordinates, parameterName=Quad 2 - BottomLeft X, sliderMax=2, sliderMin=-2, sortPriority=38}
  n47 = ScalarParameter() {defaultValue=1, group=Quad 2 Vertex Coordinates, parameterName=Quad 2 - BottomLeft Y, sliderMax=2, sliderMin=-2, sortPriority=39}
  n48 = AppendVector(A=@n46, B=@n47)
  n49 = NamedRerouteUsage()
  NamedRerouteDeclaration(None=MaterialFunctionCall(
    UVs=NamedRerouteUsage()
    vertex_0=Add(A=@n36, B=@n37)
    vertex_1=Add(A=@n40, B=@n41)
    vertex_2=Add(A=@n44, B=@n45)
    vertex_3=Add(A=@n48, B=@n49)
    Glow Min=ScalarParameter() {group=Quad 2, parameterName=Quad 2 Glow Min, sortPriority=7}
    Glow Max=ScalarParameter() {group=Quad 2, parameterName=Quad 2 Glow Max, sortPriority=8}
   ) {materialFunction=&MF_UI_WIP_Creative_Polygon_4_Vertex}.Fill) {name=Quad 2 Shape, nodeColor=<rgba 0 0 0 1>}
  n50 = ScalarParameter() {defaultValue=-1, group=Quad 3 Vertex Coordinates, parameterName=Quad 3 - TopLeft Y, sliderMax=2, sliderMin=-2, sortPriority=33}
  n51 = AppendVector(A=ScalarParameter() {defaultValue=-1, group=Quad 3 Vertex Coordinates, parameterName=Quad 3 - TopLeft X, sliderMax=2, sliderMin=-2}, B=@n50)
  n52 = NamedRerouteUsage()
  n53 = ScalarParameter() {defaultValue=1, group=Quad 3 Vertex Coordinates, parameterName=Quad 3 - TopRight X, sliderMax=2, sliderMin=-2, sortPriority=34}
  n54 = ScalarParameter() {defaultValue=-1, group=Quad 3 Vertex Coordinates, parameterName=Quad 3 - TopRight Y, sliderMax=2, sliderMin=-2, sortPriority=35}
  n55 = AppendVector(A=@n53, B=@n54)
  n56 = NamedRerouteUsage()
  n57 = ScalarParameter() {defaultValue=1, group=Quad 3 Vertex Coordinates, parameterName=Quad 3 - BottomRight X, sliderMax=2, sliderMin=-2, sortPriority=36}
  n58 = ScalarParameter() {defaultValue=1, group=Quad 3 Vertex Coordinates, parameterName=Quad 3 - BottomRight Y, sliderMax=2, sliderMin=-2, sortPriority=37}
  n59 = AppendVector(A=@n57, B=@n58)
  n60 = NamedRerouteUsage()
  n61 = ScalarParameter() {defaultValue=-1, group=Quad 3 Vertex Coordinates, parameterName=Quad 3 - BottomLeft X, sliderMax=2, sliderMin=-2, sortPriority=38}
  n62 = ScalarParameter() {defaultValue=1, group=Quad 3 Vertex Coordinates, parameterName=Quad 3 - BottomLeft Y, sliderMax=2, sliderMin=-2, sortPriority=39}
  n63 = AppendVector(A=@n61, B=@n62)
  n64 = NamedRerouteUsage()
  NamedRerouteDeclaration(None=MaterialFunctionCall(
    UVs=NamedRerouteUsage()
    vertex_0=Add(A=@n51, B=@n52)
    vertex_1=Add(A=@n55, B=@n56)
    vertex_2=Add(A=@n59, B=@n60)
    vertex_3=Add(A=@n63, B=@n64)
    Glow Min=ScalarParameter() {group=Quad 3, parameterName=Quad 3 Glow Min, sortPriority=7}
    Glow Max=ScalarParameter() {group=Quad 3, parameterName=Quad 3 Glow Max, sortPriority=8}
   ) {materialFunction=&MF_UI_WIP_Creative_Polygon_4_Vertex}.Fill) {name=Quad 3 Shape, nodeColor=<rgba 0 0 0 1>}
  n65 = MaterialFunctionCall(
   UVs=NamedRerouteUsage()
   Progress=NamedRerouteUsage()
   Rotation=ScalarParameter() {group=Quad 1, parameterName=Quad 1 Progress Rotation, sliderMax=1, sortPriority=1}
   Smooth Progress Min=ScalarParameter() {group=Quad 1, parameterName=Quad 1 Smooth Progress Min, sliderMax=1, sortPriority=2}
   Smooth Progress Max=ScalarParameter() {defaultValue=1, group=Quad 1, parameterName=Quad 1 Smooth Progress Max, sliderMax=1, sortPriority=3}
  ) {
   desc=Compute Progress / Rotation / Alpha Gradient Smoothness
   materialFunction=&MF_UI_Creative_RotativeProgress
  }
  n66 = ScalarParameter() {defaultValue=0.5, group="0. Color && Opacity", parameterName=Quad 1 Opacity, sliderMax=1, sortPriority=5}
  NamedRerouteDeclaration(None=Multiply(A=@n65.Progress, B=Saturate(None=@n66))) {name=Quad 1 Alpha, nodeColor=<rgba 0 0 0 1>}
  NamedRerouteDeclaration(None=MaterialFunctionCall() {materialFunction=&GetUserInterfaceUV}.Normalized UV) {name=Normalized UVs, nodeColor=<rgba 0 0 0 1>}
  n67 = Add(A=NamedRerouteUsage(), B=NamedRerouteUsage())
  n68 = NamedRerouteUsage()
  NamedRerouteDeclaration(None=Add(A=Add(A=@n67, B=@n68), B=NamedRerouteUsage())) {name=Final Alpha, nodeColor=<rgba 0 0 0 1>}
  n69 = MaterialFunctionCall(
   UVs=NamedRerouteUsage()
   Progress=NamedRerouteUsage()
   Rotation=ScalarParameter() {group=Quad 2, parameterName=Quad 2 Progress Rotation, sliderMax=1, sortPriority=1}
   Smooth Progress Min=ScalarParameter() {group=Quad 2, parameterName=Quad 2 Smooth Progress Min, sliderMax=1, sortPriority=2}
   Smooth Progress Max=ScalarParameter() {defaultValue=1, group=Quad 2, parameterName=Quad 2 Smooth Progress Max, sliderMax=1, sortPriority=3}
  ) {
   desc=Compute Progress / Rotation / Alpha Gradient Smoothness
   materialFunction=&MF_UI_Creative_RotativeProgress
  }
  n70 = ScalarParameter() {defaultValue=0.5, group="0. Color && Opacity", parameterName=Quad 2 Opacity, sliderMax=1, sortPriority=6}
  NamedRerouteDeclaration(None=Multiply(A=@n69.Progress, B=Saturate(None=@n70))) {name=Quad 2 Alpha, nodeColor=<rgba 0 0 0 1>}
  NamedRerouteUsage()
  NamedRerouteUsage()
  PinBase()
  PinBase()
  PinBase()
  PinBase()
 ]
}
