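# Translucent UI material for the Creative HUD that draws a tinted five-vertex polygon with adjustable opacity and a progress mask.
# Five parameter-driven vertices and glow limits feed a polygon function, with parameters for shape offset, scale and rotation.
# A remapped progress value drives a rotatable, smoothed gradient from a RotativeProgress function, which is multiplied by opacity to form the shape's alpha.
# Named reroutes link the stages.
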Material /Game/Athena/HUD/Creative/Materials/M_UI_Creative_Polygon_5_Vertices_Simpler
graph {
 properties {blendMode=BLEND_Translucent, materialDomain=MD_UI, pixelDepthOffsetMode=PDOM_Legacy, usedShadingModels=Default Lit}
 nodes [
  n1 = NamedRerouteUsage()
  n2 = NamedRerouteUsage()
  n3 = NamedRerouteUsage()
  n4 = NamedRerouteUsage()
  n5 = NamedRerouteUsage()
  n6 = NamedRerouteUsage()
  n7 = ScalarParameter() {group=2. Shape, parameterName=Shape Glow Min, sortPriority=7}
  n8 = ScalarParameter() {group=2. Shape, parameterName=Shape Glow Max, sortPriority=8}
  NamedRerouteDeclaration(None=Saturate(None=MaterialFunctionCall!(UVs=@n1, vertex_0=@n2, vertex_1=@n3, vertex_2=@n4, vertex_3=@n5, vertex_4=@n6, Outline Thickness, Glow Min=@n7, Glow Max=@n8) {materialFunction=/Game/Athena/HUD/Creative/MaterialFuntions/MF_UI_WIP_Creative_Polygon_5_Vertex}.Fill)) {name=Shape, nodeColor=<rgba 0 0 0 1>}
  n9 = VectorParameter() {
   desc="Set color, use alpha for global opacity"
   group="1. Color && Opacity"
   parameterName=Shape Color 1
   sortPriority=1
   defaultValue=<rgba 0.159 0.4845 1 1>
  }
  NamedRerouteDeclaration(None=Saturate(None=@n9.RGB)) {name=Shape Color, nodeColor=<rgba 0 0 0 1>}
  NamedRerouteDeclaration(None=AppendVector(
    A=ScalarParameter() {group=2. Shape, parameterName=Shape Offset X, sortPriority=5}
    B=ScalarParameter() {group=2. Shape, parameterName=Shape Offset Y, sortPriority=6}
   )) {name=Shape Offset, nodeColor=<rgba 0 0 0 1>}
  n10 = NamedRerouteUsage()
  n11 = AppendVector(
   A=ScalarParameter() {defaultValue=1, group=2. Shape, parameterName=Shape Scale X, sortPriority=4}
   B=ScalarParameter() {defaultValue=1, group=2. Shape, parameterName=Shape Scale Y, sortPriority=4}
  )
  NamedRerouteDeclaration(None=MaterialFunctionCall(
    UVs=MaterialFunctionCall(UVs=@n10, Texture Scale=@n11) {materialFunction=&ScaleUVsByCenter}.UVs
    "Rotation Angle (0-1)"=ScalarParameter() {group=2. Shape, parameterName=Shape Rotation, sortPriority=9}
   ) {materialFunction=&CustomRotator}.Rotated Values) {name=Shape UVs Scale, nodeColor=<rgba 0 0 0 1>}
  n12 = ScalarParameter() {defaultValue=1, group=3. Shape Vertex Coordinates, parameterName=Shape Vertex 1 X, sliderMax=2, sliderMin=-2, sortPriority=34}
  n13 = ScalarParameter() {defaultValue=-1, group=3. Shape Vertex Coordinates, parameterName=Shape Vertex 1 Y, sliderMax=2, sliderMin=-2, sortPriority=35}
  NamedRerouteDeclaration(None=Add(A=AppendVector(A=@n12, B=@n13), B=NamedRerouteUsage())) {name=Shape Vertex 1, nodeColor=<rgba 0 0 0 1>}
  n14 = ScalarParameter() {defaultValue=-1, group=3. Shape Vertex Coordinates, parameterName=Shape Vertex 4 X, sliderMax=2, sliderMin=-2, sortPriority=40}
  n15 = ScalarParameter() {defaultValue=1, group=3. Shape Vertex Coordinates, parameterName=Shape Vertex 4 Y, sliderMax=2, sliderMin=-2, sortPriority=41}
  NamedRerouteDeclaration(None=Add(A=AppendVector(A=@n14, B=@n15), B=NamedRerouteUsage())) {name=Shape Vertex 4, nodeColor=<rgba 0 0 0 1>}
  n16 = NamedRerouteUsage()
  n17 = NamedRerouteUsage()
  NamedRerouteDeclaration(None=Saturate(None=Multiply(A=@n16, B=@n17))) {name=Shape With Alpha, nodeColor=<rgba 0 0 0 1>}
  n18 = ScalarParameter() {defaultValue=1, group=3. Shape Vertex Coordinates, parameterName=Shape Vertex 2 X, sliderMax=2, sliderMin=-2, sortPriority=36}
  n19 = ScalarParameter() {defaultValue=1, group=3. Shape Vertex Coordinates, parameterName=Shape Vertex 2 Y, sliderMax=2, sliderMin=-2, sortPriority=37}
  NamedRerouteDeclaration(None=Add(A=AppendVector(A=@n18, B=@n19), B=NamedRerouteUsage())) {name=Shape Vertex 2, nodeColor=<rgba 0 0 0 1>}
  n20 = ScalarParameter() {group=3. Shape Vertex Coordinates, parameterName=Shape Vertex 3 X, sliderMax=2, sliderMin=-2, sortPriority=38}
  n21 = ScalarParameter() {defaultValue=1, group=3. Shape Vertex Coordinates, parameterName=Shape Vertex 3 Y, sliderMax=2, sliderMin=-2, sortPriority=39}
  NamedRerouteDeclaration(None=Add(A=AppendVector(A=@n20, B=@n21), B=NamedRerouteUsage())) {name=Shape Vertex 3, nodeColor=<rgba 0 0 0 1>}
  n22 = MaterialFunctionCall!(
   UVs=NamedRerouteUsage()
   Progress=NamedRerouteUsage()
   Rotation=ScalarParameter() {group=2. Shape, parameterName=Shape Progress Rotation, sliderMax=1, sortPriority=1}
   Rotation Center
   Smooth Progress Min=ScalarParameter() {group=2. Shape, parameterName=Shape Smooth Progress Min, sliderMax=1, sortPriority=2}
   Smooth Progress Max=ScalarParameter() {defaultValue=1, group=2. Shape, parameterName=Shape Smooth Progress Max, sliderMax=1, sortPriority=3}
  ) {
   desc=Compute Progress / Rotation / Alpha Gradient Smoothness
   materialFunction=/Game/Athena/HUD/Creative/MaterialFuntions/MF_UI_Creative_RotativeProgress
  }
  n23 = ScalarParameter() {defaultValue=1, group="1. Color && Opacity", parameterName=Shape Opacity, sliderMax=1, sortPriority=9}
  NamedRerouteDeclaration(None=Multiply(A=@n22.Progress, B=Saturate(None=@n23))) {name=Shape Alpha, nodeColor=<rgba 0 0 0 1>}
  NamedRerouteDeclaration(None=TextureCoordinate()) {name=UVs, nodeColor=<rgba 0 0 0 1>}
  n24 = ScalarParameter() {defaultValue=-1, group=3. Shape Vertex Coordinates, parameterName=Shape Vertex 0 X, sliderMax=2, sliderMin=-2}
  n25 = ScalarParameter() {defaultValue=-1, group=3. Shape Vertex Coordinates, parameterName=Shape Vertex 0 Y, sliderMax=2, sliderMin=-2, sortPriority=33}
  NamedRerouteDeclaration(None=Add(A=AppendVector(A=@n24, B=@n25), B=NamedRerouteUsage())) {name=Shape Vertex 0, nodeColor=<rgba 0 0 0 1>}
  NamedRerouteDeclaration(None=MaterialFunctionCall(
    Input=ScalarParameter() {defaultValue=1, group=2. Shape, parameterName=Shape Progress, sliderMax=1, sortPriority=0}
    Input Low=Constant()
    Input High=Constant() {r=1}
    Target Low=Constant() {desc="ask adrienne ", r=-1.2}
    Target High=Constant() {r=1}
   ) {materialFunction=&RemapValueRange}.Result) {name=Shape Progress Remapped, nodeColor=<rgba 0 0 0 1>}
  NamedRerouteUsage()
  NamedRerouteUsage()
 ]
}
