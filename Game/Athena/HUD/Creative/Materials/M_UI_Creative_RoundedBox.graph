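# Translucent UI material for the Creative HUD that draws a rounded box with a separately colored outline.
# A keep-aspect-ratio rounded box function turns size, corner radius, outline thickness and UI pixel size into fill and outline masks.
# Opacity parameters scale the masks, which named reroutes carry into sums for tinted color and combined coverage.
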Material /Game/Athena/HUD/Creative/Materials/M_UI_Creative_RoundedBox
graph {
 properties {blendMode=BLEND_Translucent, materialDomain=MD_UI, pixelDepthOffsetMode=PDOM_Legacy, usedShadingModels=Default Lit}
 nodes [
  n1 = VectorParameter() {group="Color && Opacity", parameterName=Color, sortPriority=1, defaultValue=<rgba 1 1 1 0.5>}
  n2 = VectorParameter() {group="Color && Opacity", parameterName=Outline Color, sortPriority=3, defaultValue=<rgba 1 0 0.4404 0.5>}
  Add(
   A=Multiply(
    A=MaterialFunctionCall(X=@n1.R, Y=@n1.R, Z=@n1.R, A=@n1.R) {materialFunction=&MakeFloat4}.Result
    B=NamedRerouteUsage()
   )
   B=Multiply(
    A=NamedRerouteUsage()
    B=MaterialFunctionCall(X=@n2.R, Y=@n2.R, Z=@n2.R, A=@n2.R) {materialFunction=&MakeFloat4}.Result
   )
  )
  Add(A=NamedRerouteUsage(), B=NamedRerouteUsage())
  n3 = VectorParameter() {group=Box Settings, parameterName=Box Corners Radius, sortPriority=2, defaultValue=<rgba 0.125 0.125 0.125 0.125>}
  n4 = MaterialFunctionCall!(
   UVs
   Pixel Size=NamedRerouteUsage()
   Box Size=MaterialFunctionCall(
    X=ScalarParameter() {defaultValue=0.5, group=Box Settings, parameterName=Box Width, sliderMax=1, sortPriority=0}
    Y=ScalarParameter() {defaultValue=0.5, group=Box Settings, parameterName=Box Height, sliderMax=1, sortPriority=1}
   ) {materialFunction=&MakeFloat2}.Result
   Corner Radiuses=MaterialFunctionCall(X=@n3.R, Y=@n3.R, Z=@n3.R, A=@n3.R) {materialFunction=&MakeFloat4}.Result
   Glow Min
   Glow Max
   Outline Thickness=ScalarParameter() {defaultValue=0.05, group=Box Settings, parameterName=Outline Thickness, sliderMax=100, sortPriority=3}
  ) {materialFunction=/Game/Athena/HUD/Creative/MaterialFuntions/MF_UI_Creative_RoundedBox_KeepAspectRatio}
  n5 = Subtract(A=@n4.Fill, B=@n4.Fill)
  n6 = ScalarParameter() {defaultValue=1, group="Color && Opacity", parameterName=Opacity, sliderMax=1, sortPriority=0}
  NamedRerouteDeclaration(None=Multiply(A=Saturate(None=@n5), B=Saturate(None=@n6))) {name=Shape, nodeColor=<rgba 0 0 0 1>}
  n7 = ScalarParameter() {defaultValue=1, group="Color && Opacity", parameterName=Outline Opacity, sliderMax=1, sortPriority=2}
  NamedRerouteDeclaration(None=Multiply(A=@n4.Outline, B=Saturate(None=@n7))) {name=Outline, nodeColor=<rgba 0 0 0 1>}
  NamedRerouteDeclaration(None=MaterialFunctionCall() {materialFunction=&GetUserInterfaceUV}.Pixel Size) {name=Pixel Size, nodeColor=<rgba 0 0 0 1>}
 ]
}
